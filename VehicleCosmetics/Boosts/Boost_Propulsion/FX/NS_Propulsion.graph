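# Niagara system for the Propulsion vehicle boost cosmetic, spraying fiery sprite exhaust backward from skeletal mesh sockets.
# Mirrored left and right fire emitters layer spawn groups with their own color, size and fade curves, while the shown smoke emitters are disabled.
# Particles partition onto sockets, fly backward and partly inherit vehicle movement.
# A frontend-preview flag and per-mode value masks retune rates, sizes and velocities.
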NiagaraSystem /VehicleCosmetics/Boosts/Boost_Propulsion/FX/NS_Propulsion
dependencies {
 usedDataInterfaces [
  &NiagaraDataInterfaceCurve
  &NiagaraDataInterfaceVectorCurve
  &NiagaraDataInterfaceArrayInt32
  &NiagaraDataInterfaceColorCurve
  &NiagaraDataInterfaceSkeletalMesh
  &NiagaraDataInterfaceArrayFloat3
 ]
 usedDynamicInputs [&SelectVectorFromArray]
 usedModules [
  &self:SystemSpawnScript.SystemScriptSource.SystemScriptGraph.NiagaraNodeAssignment_0.SetVariables_#1
  &self:SystemSpawnScript.SystemScriptSource.SystemScriptGraph.NiagaraNodeAssignment_1.SetVariables_#2
  &self:SystemSpawnScript.SystemScriptSource.SystemScriptGraph.NiagaraNodeAssignment_2.SetVariables_#3
  &SystemState
 ]
}
emitters {
 NE_VC_Boost_LeftFire {
  emitterUpdateScript [
   &EmitterState()
   &VC_FilteredSocketCount()
   &SpawnRate(
    SpawnRate=&Multiply_Float(
     A=&MakeCustomFloatFromBool(Bool=$User.bIsFrontendPreview, True Float=30, False Float=100)
     B=/VehicleCosmetics/FX/Niagara/Modules/VC_Mask_Floats(Coro Value:NiagaraFloat=1, FrontEnd Value:NiagaraFloat=1, FN Value:NiagaraFloat=0.5)
    )
    Use Spawn Probability=false
   )
   &SpawnRate(
    SpawnGroup=1
    SpawnRate=&Multiply_Float(
     A=&MakeCustomFloatFromBool(Bool=$User.bIsFrontendPreview, True Float=50, False Float=100)
     B=&VC_Mask_Floats(Coro Value:NiagaraFloat=1, FrontEnd Value:NiagaraFloat=0.5, FN Value:NiagaraFloat=0.5)
    )
    Use Spawn Probability=false
   ) {module=SpawnRate001}
   /SetVariables(
    Emitter.Base Power:NiagaraFloat=&MakeCustomFloatFromBool(Bool=$User.bIsFrontendPreview, True Float=3, False Float=2)
    Emitter.Detail Power:NiagaraFloat=&MakeCustomFloatFromBool(Bool=$User.bIsFrontendPreview, True Float=1, False Float=0.5)
   ) {isSetParameters=true, module=SetVariables_#4}
  ]
  particleSpawnScript [
   &VC_PaintedVariantsOverride() {enabled=false}
   /SetVariables(Particles.VisibilityTag=&SpawnGroup()) {isSetParameters=true, module=SetVariables_#5}
   &InitializeParticle(
    Color=&MergeHelper_ColorSwitch(
     0=&MakeCustomLinearColorFromBool(Bool=$User.bIsFrontendPreview, True Color=<rgba 2 0.42 0.1 1>, False Color=<rgba 2 0.385 0.1 1>)
     1=<rgba 1 0.6146 0.6146 1>
     2=<rgba 1 1 1 2>
     3=?
     Index=&SpawnGroup()
     Num=::NewEnumerator1
    )
    Lifetime=&MergeHelper_FloatSwitch(
     0=&MakeCustomFloatFromBool(Bool=$User.bIsFrontendPreview, True Float=0.9, False Float=1)
     1=&MakeCustomFloatFromBool(Bool=$User.bIsFrontendPreview, True Float=0.9, False Float=1)
     2=&MakeCustomFloatFromBool(Bool=$User.bIsFrontendPreview, True Float=&RandomRangeFloat(Minimum=0.45, Maximum=0.3), False Float=1)
     3=?
     Index=&SpawnGroup()
     Num=::NewEnumerator0
    )
    Sprite Rotation Mode=::NewEnumerator2
    Sprite Size Mode=::NewEnumerator3
    Sprite UV Mode=::NewEnumerator3
    Uniform Sprite Size=&MergeHelper_FloatSwitch(
     0=&MakeCustomFloatFromBool(
      Bool=$User.bIsFrontendPreview
      True Float=&RandomRangeFloat(Minimum=200, Maximum=300)
      False Float=&RandomRangeFloat(Minimum=200, Maximum=600)
     )
     1=&MakeCustomFloatFromBool(
      Bool=$User.bIsFrontendPreview
      True Float=&RandomRangeFloat(Minimum=300, Maximum=400)
      False Float=&RandomRangeFloat(Minimum=700, Maximum=650)
     )
     2=&MakeCustomFloatFromBool(
      Bool=$User.bIsFrontendPreview
      True Float=&RandomRangeFloat(Minimum=100, Maximum=150)
      False Float=&RandomRangeFloat(Minimum=400, Maximum=350)
     )
     3=?
     Index=&SpawnGroup()
     Num=::NewEnumerator0
    )
    UsePositionOffset=false
    Write Material Random=true
   )
   &PartitionParticles(Partition Mode:ENiagara_PartitionMode_v2=::NewEnumerator4)
   &SkeletalMeshLocation(
    Bone / Socket Index=$Particles.PartitionParticles.Partition
    Bone / Socket Sampling Mode=::NewEnumerator5
    MaskSpawnGroup=false
    Mesh Sampling Type=::NewEnumerator6
    Sampled Position Offset=<0 10 0>
    Skeletal Mesh=$System.SkeletalMesh
   )
   &AddVelocity(Velocity=&VC_Mask_Vectors(
     Coro Value:Vector3f=<-1500 0 0>
     FrontEnd Value:Vector3f=&MakeCustomVectorFromBool(
      Bool=$User.bIsFrontendPreview
      True Vector=<-2000 0 0>
      False Vector=&RandomRangeVector(Minimum=<-2500 0 300>, Maximum=<-2500 0 500>)
     )
     FN Value:Vector3f=<-2500 0 0>
    ))
   &CameraOffset(Camera Offset Amount=&MergeHelper_FloatSwitch(0=0, 1=-5, 2=?, 3=?, Index=&SpawnGroup(), Num=::NewEnumerator0))
   /VehicleCosmetics/FX/VC_FXParents/Modules/DLR_ProductPreviews(GameModeID:NiagaraInt32=$System.GameModeID, bPreviewVehicleSpeed:NiagaraBool=$System.bPreviewVehicleSpeed)
  ]
  particleUpdateScript [
   &ParticleState()
   &ScaleSpriteSize(
    Scale Sprite Size Mode=::NewEnumerator0
    Uniform Scale Factor=&MergeHelper_FloatSwitch(
     0=&MakeCustomFloatFromBool(
      Bool=$User.bIsFrontendPreview
      True Float=&FloatFromCurve(
       FloatCurve=DataInterface() {propertyValues {Curve {
          defaultValue=1
          keys [
           {
            arriveTangent=4.089
            arriveTangentWeight=0.0667
            interpMode=RCIM_Cubic
            leaveTangent=4.089
            leaveTangentWeight=0.1333
            tangentMode=RCTM_User
            time=0
            value=0.2
           }
           {arriveTangent=0.7313, interpMode=RCIM_Cubic, leaveTangent=0.7313, tangentMode=RCTM_User, time=0.25, value=0.8}
           <at 1 1 RCIM_Cubic>
          ]
         }}}
       Scale Curve=1.1
      )
      False Float=&FloatFromCurve(
       FloatCurve=DataInterface() {propertyValues {Curve {
          defaultValue=1
          keys [
           {arriveTangent=2.6357, arriveTangentWeight=0.0667, interpMode=RCIM_Cubic, leaveTangentWeight=0.1333, time=0, value=0.35}
           {arriveTangent=0.8373, interpMode=RCIM_Cubic, leaveTangent=0.8373, tangentMode=RCTM_User, time=0.25, value=0.8}
           {arriveTangentWeight=0.2513, interpMode=RCIM_Cubic, leaveTangentWeight=0.1393, tangentMode=RCTM_User, time=1, value=1}
          ]
         }}}
       Scale Curve=1.2
      )
     )
     1=&FloatFromCurve(
      FloatCurve=DataInterface() {propertyValues {Curve {
         defaultValue=1
         keys [
          {arriveTangent=0.8122, interpMode=RCIM_Cubic, leaveTangent=0.8122, tangentMode=RCTM_User, time=0, value=0.4}
          {interpMode=RCIM_Cubic, leaveTangent=-0.4, time=1, value=1}
         ]
        }}}
      Scale Curve=1.5
     )
     2=&FloatFromCurve(FloatCurve=DataInterface() {propertyValues {Curve {
         defaultValue=1
         keys [
          {arriveTangent=-0.1605, interpMode=RCIM_Cubic, leaveTangent=-0.1605, tangentMode=RCTM_User, time=0, value=0.4}
          {interpMode=RCIM_Cubic, leaveTangent=-0.4, time=1, value=1.2}
         ]
        }}})
     3=?
     Index=&SpawnGroup()
     Num=::NewEnumerator0
    )
   )
   &SpriteRotationRate(Rotation Rate=&RandomRangeFloat(Minimum=125, Maximum=-125))
   &DynamicMaterialParameters(
    Index 0 Param 1=1
    Index 0 Param 2=&MergeHelper_FloatSwitch(0=1.2, 1=1, 2=?, 3=?, Index=&SpawnGroup(), Num=::NewEnumerator0)
    Index 0 Param 3=&MergeHelper_FloatSwitch(0=0.1, 1=0, 2=?, 3=?, Index=&SpawnGroup(), Num=::NewEnumerator0)
    Index 2 Param 1=0.5
    Index 2 Param 2=0.5
    Index 2 Param 3=&FloatFromCurve(FloatCurve=DataInterface() {propertyValues {Curve {keys [<at 0 0.5>,<at 1 1>]}}}, Scale Curve=0.5)
    Index 2 Param 4=&FloatFromCurve(FloatCurve=DataInterface() {propertyValues {Curve {keys [<at 0 0.5>,<at 1 1>]}}}, Scale Curve=0.5)
    Index 3 Param 1=&FloatFromCurve(FloatCurve=DataInterface() {propertyValues {Curve {keys [<at 0.5 0 RCIM_Cubic>, {arriveTangent=4.0664, interpMode=RCIM_Cubic, leaveTangent=4.0664, tangentMode=RCTM_User, time=1, value=1}]}}})
    Module.Index0Float0Write=true
    Module.Index0Float1Write=true
    Module.Index0Float2Write=true
    Module.Index0Float3Write=true
    Module.Index2Float0Write=true
    Module.Index2Float1Write=true
    Module.Index2Float2Write=true
    Module.Index2Float3Write=true
    Module.Index3Float0Write=true
    Module.Index3Float1Write=true
    Module.Index3Float2Write=true
    Module.Index3Float3Write=true
    Write Parameter Index 2=::NewEnumerator1
    Write Parameter Index 3=::NewEnumerator1
   )
   &InheritSourceMovement(
    Applied Position Scale=&VectorFromFloat(Value=&FloatFromCurve(FloatCurve=DataInterface() {propertyValues {Curve {
         defaultValue=1
         keys [
          {arriveTangent=-1.7885, interpMode=RCIM_Cubic, leaveTangent=-1.7885, tangentMode=RCTM_User, time=0.05, value=1}
          {interpMode=RCIM_Cubic, tangentMode=RCTM_User, time=1, value=0}
         ]
        }}}))
    ApplyForce=false
    ApplyPosition=true
    ApplyPositionFalloff=false
   )
   &ScaleVelocity(
    Coordinate Space:ENiagaraCoordinateSpace=::Simulation
    Velocity Scale=&VectorFromFloat(Value=&FloatFromCurve(FloatCurve=DataInterface() {propertyValues {Curve {keys [{arriveTangent=-0.9718, interpMode=RCIM_Cubic, leaveTangent=-0.9718, tangentMode=RCTM_User, time=0, value=1}, <at 1 0.6 RCIM_Cubic>]}}}))
   )
   &SolveForcesAndVelocity(Clamp Velocity=false, Limit Acceleration=false)
   &ScaleColor(
    Color Scale:LinearColor=&MergeHelper_ColorSwitch(
     0=&MakeLinearColorFromVectorAndFloat(
      "Vector (RGB)"=&MakeCustomVectorFromBool(
       Bool=$User.bIsFrontendPreview
       True Vector=&Multiply_VectorByFloat(
        Vector=<1 1 1>
        Float=&FloatFromCurve(
         FloatCurve=DataInterface() {propertyValues {Curve {keys [
             {arriveTangent=-3.0335, interpMode=RCIM_Cubic, leaveTangent=-3.0335, tangentMode=RCTM_User, time=0, value=1}
             {arriveTangent=-0.1759, interpMode=RCIM_Cubic, leaveTangent=-0.1759, tangentMode=RCTM_User, time=1, value=0.1}
            ]}}}
         Scale Curve=&RandomRangeFloat(Minimum=1.5, Maximum=1)
        )
       )
       False Vector=&Multiply_VectorByFloat(
        Vector=&MakeVector(
         X=1
         Y=&FloatFromCurve(FloatCurve=DataInterface() {propertyValues {Curve {keys [{arriveTangent=-1.0104, interpMode=RCIM_Cubic, leaveTangent=-1.0104, tangentMode=RCTM_User, time=0, value=1}, <at 0.5 0.8 RCIM_Cubic>]}}})
         Z=1
        )
        Float=&FloatFromCurve(
         FloatCurve=DataInterface() {propertyValues {Curve {keys [
             {arriveTangent=-3.0335, interpMode=RCIM_Cubic, leaveTangent=-3.0335, tangentMode=RCTM_User, time=0, value=1}
             {arriveTangent=-0.1759, interpMode=RCIM_Cubic, leaveTangent=-0.1759, tangentMode=RCTM_User, time=1, value=0.1}
            ]}}}
         Scale Curve=&RandomRangeFloat(Minimum=1.5, Maximum=1)
        )
       )
      )
      "Float (Alpha)"=&FloatFromCurve(FloatCurve=DataInterface() {propertyValues {Curve {keys [
           {arriveTangent=0.3551, interpMode=RCIM_Cubic, leaveTangent=0.3551, tangentMode=RCTM_User, time=0, value=0}
           {arriveTangentWeight=0.1167, interpMode=RCIM_Cubic, leaveTangentWeight=0.4218, tangentMode=RCTM_User, time=0.13, value=1.5}
           {arriveTangentWeight=0.3841, interpMode=RCIM_Cubic, leaveTangentWeight=0.1635, tangentMode=RCTM_User, time=0.6, value=1.5}
           <at 0.8 0 RCIM_Cubic>
          ]}}})
     )
     1=&MakeLinearColorFromVectorAndFloat(
      "Vector (RGB)"=&VectorFromFloat(Value=0.01)
      "Float (Alpha)"=&FloatFromCurve(
       FloatCurve=DataInterface() {propertyValues {Curve {keys [
           {interpMode=RCIM_Cubic, leaveTangentWeight=0.0667, tangentMode=RCTM_User, time=0.15, value=0}
           {arriveTangentWeight=0.1491, interpMode=RCIM_Cubic, leaveTangentWeight=0.2236, tangentMode=RCTM_User, time=0.3, value=1}
           {arriveTangentWeight=0.1944, interpMode=RCIM_Cubic, leaveTangentWeight=0.1944, tangentMode=RCTM_User, time=0.6, value=1}
           {arriveTangentWeight=0.1, interpMode=RCIM_Cubic, tangentMode=RCTM_User, time=1, value=0}
          ]}}}
       Scale Curve=1.2
      )
     )
     2=&MakeCustomLinearColorFromBool(
      Bool=$User.bIsFrontendPreview
      True Color=&MakeLinearColorFromVectorAndFloat(
       "Vector (RGB)"=&VectorFromCurve(VectorCurve=DataInterface() {propertyValues {
          XCurve {keys [<at 0 1.1>,<at 0.3888 1>,<at 0.9 0.4>]}
          YCurve {keys [
            {arriveTangent=-0.3418, interpMode=RCIM_Cubic, leaveTangent=-0.3418, tangentMode=RCTM_User, time=0, value=1.1}
            {arriveTangent=-0.7778, interpMode=RCIM_Cubic, leaveTangent=-0.7778, time=0.5149, value=0.6753}
            {arriveTangent=-1.5559, interpMode=RCIM_Cubic, leaveTangent=-1.5559, tangentMode=RCTM_User, time=0.9, value=0.4}
           ]}
          ZCurve {keys [<at 0 1.1>,<at 0.5149 0.7349>,<at 0.9 0.4>]}
         }})
       "Float (Alpha)"=&FloatFromCurve(FloatCurve=DataInterface() {propertyValues {Curve {keys [<at 0.1 0>,<at 0.15 1>,<at 1 0>]}}})
      )
      False Color=&MakeLinearColorFromVectorAndFloat(
       "Vector (RGB)"=&VectorFromCurve(VectorCurve=DataInterface() {propertyValues {
          XCurve {keys [<at 0 1.1>,<at 0.3888 1>,<at 0.9 0.4>]}
          YCurve {keys [
            {arriveTangent=-0.3418, interpMode=RCIM_Cubic, leaveTangent=-0.3418, tangentMode=RCTM_User, time=0, value=1.1}
            {arriveTangent=-0.7778, interpMode=RCIM_Cubic, leaveTangent=-0.7778, time=0.5149, value=0.6753}
            {arriveTangent=-1.5559, interpMode=RCIM_Cubic, leaveTangent=-1.5559, tangentMode=RCTM_User, time=0.9, value=0.4}
           ]}
          ZCurve {keys [<at 0 1.1>,<at 0.5149 0.7349>,<at 0.9 0.4>]}
         }})
       "Float (Alpha)"=&FloatFromCurve(FloatCurve=DataInterface() {propertyValues {Curve {keys [<at 0.1 0>,<at 0.15 1>,<at 1 1>]}}})
      )
     )
     3=?
     Index=&SpawnGroup()
     Num=::NewEnumerator0
    )
    Scale Mode=3
   )
  ]
  rendererClasses [&NiagaraSpriteRendererProperties]
  renderers [
   {rendererClass=&NiagaraSpriteRendererProperties}
   {rendererClass=&NiagaraSpriteRendererProperties, rendererIndex=1}
  ]
 }
 NE_VC_Boost_RightFire {
  emitterUpdateScript [
   &EmitterState()
   &VC_FilteredSocketCount()
   &SpawnRate(
    SpawnRate=&Multiply_Float(
     A=&MakeCustomFloatFromBool(Bool=$User.bIsFrontendPreview, True Float=30, False Float=100)
     B=&VC_Mask_Floats(Coro Value:NiagaraFloat=1, FrontEnd Value:NiagaraFloat=1, FN Value:NiagaraFloat=0.5)
    )
    Use Spawn Probability=false
   )
   &SpawnRate(
    SpawnGroup=1
    SpawnRate=&Multiply_Float(
     A=&MakeCustomFloatFromBool(Bool=$User.bIsFrontendPreview, True Float=50, False Float=100)
     B=&VC_Mask_Floats(Coro Value:NiagaraFloat=1, FrontEnd Value:NiagaraFloat=0.5, FN Value:NiagaraFloat=0.5)
    )
    Use Spawn Probability=false
   ) {module=SpawnRate001}
   /SetVariables(
    Emitter.Base Power:NiagaraFloat=&MakeCustomFloatFromBool(Bool=$User.bIsFrontendPreview, True Float=3, False Float=2)
    Emitter.Detail Power:NiagaraFloat=&MakeCustomFloatFromBool(Bool=$User.bIsFrontendPreview, True Float=1, False Float=0.5)
   ) {isSetParameters=true, module=SetVariables_#4}
  ]
  particleSpawnScript [
   &VC_PaintedVariantsOverride() {enabled=false}
   /SetVariables(Particles.VisibilityTag=&SpawnGroup()) {isSetParameters=true, module=SetVariables_#5}
   &InitializeParticle(
    Color=&MergeHelper_ColorSwitch(
     0=&MakeCustomLinearColorFromBool(Bool=$User.bIsFrontendPreview, True Color=<rgba 2 0.42 0.1 1>, False Color=<rgba 2 0.385 0.1 1>)
     1=<rgba 1 0.6146 0.6146 1>
     2=<rgba 1 1 1 2>
     3=?
     Index=&SpawnGroup()
     Num=::NewEnumerator1
    )
    Lifetime=&MergeHelper_FloatSwitch(
     0=&MakeCustomFloatFromBool(Bool=$User.bIsFrontendPreview, True Float=0.9, False Float=1)
     1=&MakeCustomFloatFromBool(Bool=$User.bIsFrontendPreview, True Float=0.9, False Float=1)
     2=&MakeCustomFloatFromBool(Bool=$User.bIsFrontendPreview, True Float=&RandomRangeFloat(Minimum=0.45, Maximum=0.3), False Float=1)
     3=?
     Index=&SpawnGroup()
     Num=::NewEnumerator0
    )
    Sprite Rotation Mode=::NewEnumerator2
    Sprite Size Mode=::NewEnumerator3
    Uniform Sprite Size=&MergeHelper_FloatSwitch(
     0=&MakeCustomFloatFromBool(
      Bool=$User.bIsFrontendPreview
      True Float=&RandomRangeFloat(Minimum=200, Maximum=300)
      False Float=&RandomRangeFloat(Minimum=200, Maximum=600)
     )
     1=&MakeCustomFloatFromBool(
      Bool=$User.bIsFrontendPreview
      True Float=&RandomRangeFloat(Minimum=300, Maximum=400)
      False Float=&RandomRangeFloat(Minimum=700, Maximum=650)
     )
     2=&MakeCustomFloatFromBool(
      Bool=$User.bIsFrontendPreview
      True Float=&RandomRangeFloat(Minimum=100, Maximum=150)
      False Float=&RandomRangeFloat(Minimum=400, Maximum=350)
     )
     3=?
     Index=&SpawnGroup()
     Num=::NewEnumerator0
    )
    UsePositionOffset=false
    Write Material Random=true
   )
   &PartitionParticles(A=0, Partition Mode:ENiagara_PartitionMode_v2=::NewEnumerator4)
   &SkeletalMeshLocation(
    Bone / Socket Index=$Particles.PartitionParticles.Partition
    Bone / Socket Sampling Mode=::NewEnumerator5
    MaskSpawnGroup=false
    Mesh Sampling Type=::NewEnumerator6
    Sampled Position Offset=<0 -10 0>
    Skeletal Mesh=$System.SkeletalMesh
   )
   &AddVelocity(Velocity=&VC_Mask_Vectors(
     Coro Value:Vector3f=<-1500 0 0>
     FrontEnd Value:Vector3f=&MakeCustomVectorFromBool(
      Bool=$User.bIsFrontendPreview
      True Vector=<-2000 0 0>
      False Vector=&RandomRangeVector(Minimum=<-2500 0 300>, Maximum=<-2500 0 500>)
     )
     FN Value:Vector3f=<-2500 0 0>
    ))
   &CameraOffset(Camera Offset Amount=&MergeHelper_FloatSwitch(0=0, 1=-5, 2=?, 3=?, Index=&SpawnGroup(), Num=::NewEnumerator0))
   &DLR_ProductPreviews(GameModeID:NiagaraInt32=$System.GameModeID, bPreviewVehicleSpeed:NiagaraBool=$System.bPreviewVehicleSpeed)
  ]
  particleUpdateScript [
   &ParticleState()
   &ScaleSpriteSize(
    Scale Sprite Size Mode=::NewEnumerator0
    Uniform Scale Factor=&MergeHelper_FloatSwitch(
     0=&MakeCustomFloatFromBool(
      Bool=$User.bIsFrontendPreview
      True Float=&FloatFromCurve(
       FloatCurve=DataInterface() {propertyValues {Curve {
          defaultValue=1
          keys [
           {
            arriveTangent=4.089
            arriveTangentWeight=0.0667
            interpMode=RCIM_Cubic
            leaveTangent=4.089
            leaveTangentWeight=0.1333
            tangentMode=RCTM_User
            time=0
            value=0.2
           }
           {arriveTangent=0.7313, interpMode=RCIM_Cubic, leaveTangent=0.7313, tangentMode=RCTM_User, time=0.25, value=0.8}
           <at 1 1 RCIM_Cubic>
          ]
         }}}
       Scale Curve=1.1
      )
      False Float=&FloatFromCurve(
       FloatCurve=DataInterface() {propertyValues {Curve {
          defaultValue=1
          keys [
           {arriveTangent=2.6357, arriveTangentWeight=0.0667, interpMode=RCIM_Cubic, leaveTangentWeight=0.1333, time=0, value=0.35}
           {arriveTangent=0.8373, interpMode=RCIM_Cubic, leaveTangent=0.8373, tangentMode=RCTM_User, time=0.25, value=0.8}
           {arriveTangentWeight=0.2513, interpMode=RCIM_Cubic, leaveTangentWeight=0.1393, tangentMode=RCTM_User, time=1, value=1}
          ]
         }}}
       Scale Curve=1.2
      )
     )
     1=&FloatFromCurve(
      FloatCurve=DataInterface() {propertyValues {Curve {
         defaultValue=1
         keys [
          {arriveTangent=0.8122, interpMode=RCIM_Cubic, leaveTangent=0.8122, tangentMode=RCTM_User, time=0, value=0.4}
          {interpMode=RCIM_Cubic, leaveTangent=-0.4, time=1, value=1}
         ]
        }}}
      Scale Curve=1.5
     )
     2=&FloatFromCurve(FloatCurve=DataInterface() {propertyValues {Curve {
         defaultValue=1
         keys [
          {arriveTangent=-0.1605, interpMode=RCIM_Cubic, leaveTangent=-0.1605, tangentMode=RCTM_User, time=0, value=0.4}
          {interpMode=RCIM_Cubic, leaveTangent=-0.4, time=1, value=1.2}
         ]
        }}})
     3=?
     Index=&SpawnGroup()
     Num=::NewEnumerator0
    )
   )
   &SpriteRotationRate(Rotation Rate=&RandomRangeFloat(Minimum=125, Maximum=-125))
   &DynamicMaterialParameters(
    Index 0 Param 1=1
    Index 0 Param 2=&MergeHelper_FloatSwitch(0=1.2, 1=1, 2=?, 3=?, Index=&SpawnGroup(), Num=::NewEnumerator0)
    Index 0 Param 3=&MergeHelper_FloatSwitch(0=0.1, 1=0, 2=?, 3=?, Index=&SpawnGroup(), Num=::NewEnumerator0)
    Index 2 Param 1=0.5
    Index 2 Param 2=0.5
    Index 2 Param 3=&FloatFromCurve(FloatCurve=DataInterface() {propertyValues {Curve {keys [<at 0 0.5>,<at 1 1>]}}}, Scale Curve=0.5)
    Index 2 Param 4=&FloatFromCurve(FloatCurve=DataInterface() {propertyValues {Curve {keys [<at 0 0.5>,<at 1 1>]}}}, Scale Curve=0.5)
    Index 3 Param 1=&FloatFromCurve(FloatCurve=DataInterface() {propertyValues {Curve {keys [<at 0.5 0 RCIM_Cubic>, {arriveTangent=4.0664, interpMode=RCIM_Cubic, leaveTangent=4.0664, tangentMode=RCTM_User, time=1, value=1}]}}})
    Module.Index0Float0Write=true
    Module.Index0Float1Write=true
    Module.Index0Float2Write=true
    Module.Index0Float3Write=true
    Module.Index2Float0Write=true
    Module.Index2Float1Write=true
    Module.Index2Float2Write=true
    Module.Index2Float3Write=true
    Module.Index3Float0Write=true
    Module.Index3Float1Write=true
    Module.Index3Float2Write=true
    Module.Index3Float3Write=true
    Write Parameter Index 2=::NewEnumerator1
    Write Parameter Index 3=::NewEnumerator1
   )
   &InheritSourceMovement(
    Applied Position Scale=&VectorFromFloat(Value=&FloatFromCurve(FloatCurve=DataInterface() {propertyValues {Curve {
         defaultValue=1
         keys [
          {arriveTangent=-1.7885, interpMode=RCIM_Cubic, leaveTangent=-1.7885, tangentMode=RCTM_User, time=0.05, value=1}
          {interpMode=RCIM_Cubic, tangentMode=RCTM_User, time=1, value=0}
         ]
        }}}))
    ApplyForce=false
    ApplyPosition=true
    ApplyPositionFalloff=false
   )
   &ScaleVelocity(
    Coordinate Space:ENiagaraCoordinateSpace=::Simulation
    Velocity Scale=&VectorFromFloat(Value=&FloatFromCurve(FloatCurve=DataInterface() {propertyValues {Curve {keys [{arriveTangent=-0.9718, interpMode=RCIM_Cubic, leaveTangent=-0.9718, tangentMode=RCTM_User, time=0, value=1}, <at 1 0.6 RCIM_Cubic>]}}}))
   )
   &SolveForcesAndVelocity(Clamp Velocity=false, Limit Acceleration=false)
   &ScaleColor(
    Color Scale:LinearColor=&MergeHelper_ColorSwitch(
     0=&MakeLinearColorFromVectorAndFloat(
      "Vector (RGB)"=&MakeCustomVectorFromBool(
       Bool=$User.bIsFrontendPreview
       True Vector=&Multiply_VectorByFloat(
        Vector=<1 1 1>
        Float=&FloatFromCurve(
         FloatCurve=DataInterface() {propertyValues {Curve {keys [
             {arriveTangent=-3.0335, interpMode=RCIM_Cubic, leaveTangent=-3.0335, tangentMode=RCTM_User, time=0, value=1}
             {arriveTangent=-0.1759, interpMode=RCIM_Cubic, leaveTangent=-0.1759, tangentMode=RCTM_User, time=1, value=0.1}
            ]}}}
         Scale Curve=&RandomRangeFloat(Minimum=1.5, Maximum=1)
        )
       )
       False Vector=&Multiply_VectorByFloat(
        Vector=&MakeVector(
         X=1
         Y=&FloatFromCurve(FloatCurve=DataInterface() {propertyValues {Curve {keys [{arriveTangent=-1.0104, interpMode=RCIM_Cubic, leaveTangent=-1.0104, tangentMode=RCTM_User, time=0, value=1}, <at 0.5 0.8 RCIM_Cubic>]}}})
         Z=1
        )
        Float=&FloatFromCurve(
         FloatCurve=DataInterface() {propertyValues {Curve {keys [
             {arriveTangent=-3.0335, interpMode=RCIM_Cubic, leaveTangent=-3.0335, tangentMode=RCTM_User, time=0, value=1}
             {arriveTangent=-0.1759, interpMode=RCIM_Cubic, leaveTangent=-0.1759, tangentMode=RCTM_User, time=1, value=0.1}
            ]}}}
         Scale Curve=&RandomRangeFloat(Minimum=1.5, Maximum=1)
        )
       )
      )
      "Float (Alpha)"=&FloatFromCurve(FloatCurve=DataInterface() {propertyValues {Curve {keys [
           {arriveTangent=0.3551, interpMode=RCIM_Cubic, leaveTangent=0.3551, tangentMode=RCTM_User, time=0, value=0}
           {arriveTangentWeight=0.1167, interpMode=RCIM_Cubic, leaveTangentWeight=0.4218, tangentMode=RCTM_User, time=0.13, value=1.5}
           {arriveTangentWeight=0.3841, interpMode=RCIM_Cubic, leaveTangentWeight=0.1635, tangentMode=RCTM_User, time=0.6, value=1.5}
           <at 0.8 0 RCIM_Cubic>
          ]}}})
     )
     1=&MakeLinearColorFromVectorAndFloat(
      "Vector (RGB)"=&VectorFromFloat(Value=0.01)
      "Float (Alpha)"=&FloatFromCurve(
       FloatCurve=DataInterface() {propertyValues {Curve {keys [
           {interpMode=RCIM_Cubic, leaveTangentWeight=0.0667, tangentMode=RCTM_User, time=0.15, value=0}
           {arriveTangentWeight=0.1491, interpMode=RCIM_Cubic, leaveTangentWeight=0.2236, tangentMode=RCTM_User, time=0.3, value=1}
           {arriveTangentWeight=0.1944, interpMode=RCIM_Cubic, leaveTangentWeight=0.1944, tangentMode=RCTM_User, time=0.6, value=1}
           {arriveTangentWeight=0.1, interpMode=RCIM_Cubic, tangentMode=RCTM_User, time=1, value=0}
          ]}}}
       Scale Curve=1.2
      )
     )
     2=&MakeCustomLinearColorFromBool(
      Bool=$User.bIsFrontendPreview
      True Color=&MakeLinearColorFromVectorAndFloat(
       "Vector (RGB)"=&VectorFromCurve(VectorCurve=DataInterface() {propertyValues {
          XCurve {keys [<at 0 1.1>,<at 0.3888 1>,<at 0.9 0.4>]}
          YCurve {keys [
            {arriveTangent=-0.3418, interpMode=RCIM_Cubic, leaveTangent=-0.3418, tangentMode=RCTM_User, time=0, value=1.1}
            {arriveTangent=-0.7778, interpMode=RCIM_Cubic, leaveTangent=-0.7778, time=0.5149, value=0.6753}
            {arriveTangent=-1.5559, interpMode=RCIM_Cubic, leaveTangent=-1.5559, tangentMode=RCTM_User, time=0.9, value=0.4}
           ]}
          ZCurve {keys [<at 0 1.1>,<at 0.5149 0.7349>,<at 0.9 0.4>]}
         }})
       "Float (Alpha)"=&FloatFromCurve(FloatCurve=DataInterface() {propertyValues {Curve {keys [<at 0.1 0>,<at 0.15 1>,<at 1 0>]}}})
      )
      False Color=&MakeLinearColorFromVectorAndFloat(
       "Vector (RGB)"=&VectorFromCurve(VectorCurve=DataInterface() {propertyValues {
          XCurve {keys [<at 0 1.1>,<at 0.3888 1>,<at 0.9 0.4>]}
          YCurve {keys [
            {arriveTangent=-0.3418, interpMode=RCIM_Cubic, leaveTangent=-0.3418, tangentMode=RCTM_User, time=0, value=1.1}
            {arriveTangent=-0.7778, interpMode=RCIM_Cubic, leaveTangent=-0.7778, time=0.5149, value=0.6753}
            {arriveTangent=-1.5559, interpMode=RCIM_Cubic, leaveTangent=-1.5559, tangentMode=RCTM_User, time=0.9, value=0.4}
           ]}
          ZCurve {keys [<at 0 1.1>,<at 0.5149 0.7349>,<at 0.9 0.4>]}
         }})
       "Float (Alpha)"=&FloatFromCurve(FloatCurve=DataInterface() {propertyValues {Curve {keys [<at 0.1 0>,<at 0.15 1>,<at 1 1>]}}})
      )
     )
     3=?
     Index=&SpawnGroup()
     Num=::NewEnumerator0
    )
    Scale Mode=3
   )
  ]
  rendererClasses [&NiagaraSpriteRendererProperties]
  renderers [
   {rendererClass=&NiagaraSpriteRendererProperties}
   {rendererClass=&NiagaraSpriteRendererProperties, rendererIndex=1}
  ]
 }
 NE_VC_Boost_Smoke {
  bEnabled=false
  emitterUpdateScript [
   &EmitterState()
   &VC_FilteredSocketCount()
   &SpawnRate(
    SpawnRate=&Multiply_Float(
     A=150
     B=&VC_Mask_Floats(Coro Value:NiagaraFloat=1, FrontEnd Value:NiagaraFloat=0.5, FN Value:NiagaraFloat=0.5)
    )
    Use Spawn Probability=false
   )
   &SpawnRate(
    SpawnGroup=1
    SpawnRate=&Multiply_Float(
     A=200
     B=&VC_Mask_Floats(Coro Value:NiagaraFloat=1, FrontEnd Value:NiagaraFloat=0.5, FN Value:NiagaraFloat=0.5)
    )
    Use Spawn Probability=false
   ) {module=SpawnRate001}
   /SetVariables(Emitter.Detail Power:NiagaraFloat=&MakeCustomFloatFromBool(Bool=$User.bIsFrontendPreview, True Float=1, False Float=0.5)) {isSetParameters=true, module=SetVariables_#4}
  ]
  particleSpawnScript [
   &VC_PaintedVariantsOverride() {enabled=false}
   &InitializeParticle(
    Color=&MergeHelper_ColorSwitch(
     0=&MakeCustomLinearColorFromBool(Bool=$User.bIsFrontendPreview, True Color=<rgba 1.2 0.2016 0.048 10>, False Color=<rgba 3 0.489 0.21 10>)
     1=<rgba 0.1 0.0191 0.01 1>
     2=?
     3=?
     Index=&SpawnGroup()
     Num=::NewEnumerator0
    )
    Lifetime=&MergeHelper_FloatSwitch(
     0=&MakeCustomFloatFromBool(Bool=$User.bIsFrontendPreview, True Float=0.6, False Float=1)
     1=&MakeCustomFloatFromBool(Bool=$User.bIsFrontendPreview, True Float=0.7, False Float=1.05)
     2=&RandomRangeFloat(Minimum=0.1, Maximum=0.4)
     3=?
     Index=&SpawnGroup()
     Num=::NewEnumerator0
    )
    Sprite Rotation Mode=::NewEnumerator2
    Sprite Size Mode=::NewEnumerator3
    Uniform Sprite Size=&MergeHelper_FloatSwitch(
     0=&MakeCustomFloatFromBool(
      Bool=$User.bIsFrontendPreview
      True Float=&RandomRangeFloat(Minimum=300, Maximum=400)
      False Float=&RandomRangeFloat(Minimum=400, Maximum=350)
     )
     1=&MakeCustomFloatFromBool(
      Bool=$User.bIsFrontendPreview
      True Float=&RandomRangeFloat(Minimum=800, Maximum=600)
      False Float=&RandomRangeFloat(Minimum=600, Maximum=700)
     )
     2=?
     3=?
     Index=&SpawnGroup()
     Num=::NewEnumerator0
    )
    UsePositionOffset=false
    Write Material Random=true
   )
   &PartitionParticles(Number Of Partitions=$Emitter.PortCount)
   &SkeletalMeshLocation(
    Bone / Socket Index=$Particles.PartitionParticles.Partition
    Bone / Socket Sampling Mode=::NewEnumerator5
    MaskSpawnGroup=false
    Mesh Sampling Type=::NewEnumerator6
    Sampled Position Offset=/Niagara/DynamicInputs/TypeConversions/MakeCustomPositionFromBool(Bool:NiagaraBool=$User.bIsFrontendPreview, True Position:NiagaraPosition=<0 0 -10>, False Position:NiagaraPosition=<0 0 25>)
    Skeletal Mesh=$System.SkeletalMesh
   )
   &AddVelocity(
    Cone Angle=5
    Cone Axis=&RandomRangeVector(Minimum=<1 -0.12 -0.02>, Maximum=<1 0.12 -0.02>)
    Module.UseConeFalloff=false
    Module.UseVelDistribution=false
    Override Seed=false
    Velocity Mode=::NewEnumerator2
    Velocity Speed=&MakeFloatFromVector(
     Vector=&VC_Mask_Vectors(Coro Value:Vector3f=<-400 0 0>, FrontEnd Value:Vector3f=<-1200 0 0>, FN Value:Vector3f=<-1200 0 0>)
     Channel=::NewEnumerator0
    )
   )
   &ShapeLocation(
    Box Size=&MergeHelper_VectorSwitch(0=<1 80 1>, 1=<1 100 1>, 2=?, 3=?, Index=&SpawnGroup(), Num=::NewEnumerator0)
    MaskSpawnGroup=false
    Override Seed=false
    Shape Primitive=::NewEnumerator2
    Surface Only=false
    UseOwnerScale=true
   )
   &CameraOffset(Camera Offset Amount=&MergeHelper_FloatSwitch(0=50, 1=-5, 2=?, 3=?, Index=&SpawnGroup(), Num=::NewEnumerator0))
   /SetVariables(Particles.VisibilityTag=$Emitter.SpawnGroup) {isSetParameters=true, module=SetVariables_#5}
   &DLR_ProductPreviews(GameModeID:NiagaraInt32=$System.GameModeID, bPreviewVehicleSpeed:NiagaraBool=$System.bPreviewVehicleSpeed)
  ]
  particleUpdateScript [
   &ParticleState()
   &ScaleSpriteSize(
    Scale Sprite Size Mode=::NewEnumerator0
    Uniform Scale Factor=&MakeCustomFloatFromBool(
     Bool=$User.bIsFrontendPreview
     True Float=&FloatFromCurve(FloatCurve=DataInterface() {propertyValues {Curve {
         defaultValue=1
         keys [{arriveTangent=1.4488, interpMode=RCIM_Cubic, leaveTangent=1.4488, tangentMode=RCTM_User, time=0.1, value=0.5}, <at 1 0.8 RCIM_Cubic>]
        }}})
     False Float=&FloatFromCurve(FloatCurve=DataInterface() {propertyValues {Curve {
         defaultValue=1
         keys [
          {arriveTangent=-0.1605, interpMode=RCIM_Cubic, leaveTangent=-0.1605, tangentMode=RCTM_User, time=0.1, value=0.6}
          {arriveTangent=0.6443, interpMode=RCIM_Cubic, leaveTangent=0.6443, time=0.4763, value=0.715}
          {arriveTangent=0.7351, interpMode=RCIM_Cubic, leaveTangent=0.7351, time=0.6802, value=0.9738}
          <at 1 1.1 RCIM_Cubic>
         ]
        }}})
    )
   )
   &SpriteRotationRate(Rotation Rate=&RandomRangeFloat(Minimum=125, Maximum=-125))
   &DynamicMaterialParameters(
    Index 0 Param 2=&MergeHelper_FloatSwitch(
     0=&FloatFromCurve(FloatCurve=DataInterface() {propertyValues {Curve {defaultValue=0, keys [{tangentMode=RCTM_User, time=0, value=1}, {tangentMode=RCTM_User, time=1, value=0}]}}})
     1=&FloatFromCurve(FloatCurve=DataInterface() {propertyValues {Curve {defaultValue=1, keys [{tangentMode=RCTM_User, time=0, value=0}, {tangentMode=RCTM_User, time=1, value=0.5}]}}})
     2=?
     3=?
     Index=&SpawnGroup()
     Num=::NewEnumerator0
    )
    Index 0 Param 3=&MergeHelper_FloatSwitch(0=0.2, 1=0.1, 2=?, 3=?, Index=&SpawnGroup(), Num=::NewEnumerator0)
    Index 0 Param 4=&MergeHelper_FloatSwitch(
     0=&FloatFromCurve(
      FloatCurve=DataInterface() {propertyValues {Curve {defaultValue=1, keys [{tangentMode=RCTM_User, time=0, value=0}, {tangentMode=RCTM_User, time=1, value=1}]}}}
      Scale Curve=0.1
     )
     1=0.05
     2=?
     3=?
     Index=&SpawnGroup()
     Num=::NewEnumerator0
    )
    Index 2 Param 1=0.5
    Index 2 Param 2=0.5
    Index 2 Param 3=&FloatFromCurve(FloatCurve=DataInterface() {propertyValues {Curve {keys [<at 0 0.5>,<at 1 1>]}}}, Scale Curve=0.5)
    Index 2 Param 4=&FloatFromCurve(FloatCurve=DataInterface() {propertyValues {Curve {keys [<at 0 0.5>,<at 1 1>]}}}, Scale Curve=0.5)
    Module.Index0Float0Write=true
    Module.Index0Float1Write=true
    Module.Index0Float2Write=true
    Module.Index0Float3Write=true
    Module.Index2Float0Write=true
    Module.Index2Float1Write=true
    Module.Index2Float2Write=true
    Module.Index2Float3Write=true
    Write Parameter Index 2=::NewEnumerator1
   )
   &InheritSourceMovement(
    Applied Position Scale=&VectorFromFloat(Value=&FloatFromCurve(FloatCurve=DataInterface() {propertyValues {Curve {
         defaultValue=1
         keys [
          {arriveTangent=-1.7885, interpMode=RCIM_Cubic, leaveTangent=-1.7885, tangentMode=RCTM_User, time=0.05, value=1}
          {interpMode=RCIM_Cubic, tangentMode=RCTM_User, time=1, value=0}
         ]
        }}}))
    ApplyForce=false
    ApplyPosition=true
    ApplyPositionFalloff=false
   )
   &SolveForcesAndVelocity(Clamp Velocity=false, Limit Acceleration=false)
   &ScaleColor(
    Scale Alpha=&MergeHelper_FloatSwitch(
     0=&FloatFromCurve(FloatCurve=DataInterface() {propertyValues {Curve {keys [<at 0.03 0>,<at 0.25 1>,<at 0.8 1>,<at 1 0>]}}})
     1=&FloatFromCurve(FloatCurve=DataInterface() {propertyValues {Curve {keys [
          <at 0.15 0 RCIM_Cubic>
          {arriveTangent=-1.2555, interpMode=RCIM_Cubic, leaveTangent=-1.2555, tangentMode=RCTM_User, time=0.5, value=1}
          <at 1 1 RCIM_Cubic>
         ]}}})
     2=?
     3=?
     Index=&SpawnGroup()
     Num=::NewEnumerator0
    )
    Scale RGB=&VectorFromFloat(Value=&MergeHelper_FloatSwitch(
      0=&MakeCustomFloatFromBool(
       Bool=$User.bIsFrontendPreview
       True Float=&MakeFloatFromVector(
        Vector=&VectorFromCurve(VectorCurve=DataInterface() {propertyValues {
           XCurve {keys [<at 0 1>,<at 0.2 3.5>,<at 0.45 0.8>,<at 1 0.5>]}
           YCurve {keys [<at 0 1>,<at 0.2 1.5>,<at 0.3 1.5>,<at 0.55 1>,<at 1 0.5>]}
           ZCurve {keys [<at 0 1>,<at 0.2 1.5>,<at 0.3 1.5>,<at 0.55 1>,<at 1 0.5>]}
          }})
        Channel=::NewEnumerator0
       )
       False Float=&FloatFromCurve(FloatCurve=DataInterface() {propertyValues {Curve {keys [
            {arriveTangent=-2.6037, interpMode=RCIM_Cubic, leaveTangent=-2.6037, tangentMode=RCTM_User, time=0, value=2}
            {arriveTangent=-1.9576, interpMode=RCIM_Cubic, leaveTangent=-1.9576, tangentMode=RCTM_User, time=1, value=0}
           ]}}})
      )
      1=&FloatFromCurve(FloatCurve=DataInterface() {propertyValues {Curve {keys [<at 0 0.3>,<at 1 0.1>]}}})
      2=?
      3=?
      Index=&SpawnGroup()
      Num=::NewEnumerator0
     ))
    ScaleA=true
    ScaleRGB=true
   )
   &ScaleColor(
    Scale RGB=&MergeHelper_VectorSwitch(
     0=&MakeCustomVectorFromBool(
      Bool=$User.bIsFrontendPreview
      True Vector=&VectorFromCurve(VectorCurve=DataInterface() {propertyValues {
         XCurve {keys [<at 0.1 3 RCIM_Cubic>,<at 1 0.5 RCIM_Cubic>]}
         YCurve {keys [<at 0.1 3 RCIM_Cubic>,<at 1 0.5 RCIM_Cubic>]}
         ZCurve {keys [<at 0.1 3 RCIM_Cubic>,<at 1 0.5 RCIM_Cubic>]}
        }})
      False Vector=&VectorFromCurve(VectorCurve=DataInterface() {propertyValues {
         XCurve {keys [<at 0 1>,<at 0.2 3.5>,<at 0.45 0.8>,<at 1 0.5>]}
         YCurve {keys [<at 0 1>,<at 0.2 1.5>,<at 0.3 1.5>,<at 0.55 1>,<at 1 0.5>]}
         ZCurve {keys [<at 0 1>,<at 0.2 1.5>,<at 0.3 1.5>,<at 0.55 1>,<at 1 0.5>]}
        }})
     )
     1=&VectorFromCurve(VectorCurve=DataInterface() {propertyValues {XCurve {keys [<at 0 0.3>,<at 1 0.1>]}, YCurve {keys [<at 0 0.3>,<at 1 0.1>]}, ZCurve {keys [<at 0 0.3>,<at 1 0.1>]}}})
     2=?
     3=?
     Index=&SpawnGroup()
     Num=::NewEnumerator0
    )
    ScaleA=false
    ScaleRGB=true
   ) {enabled=false, module=ScaleColor001}
  ]
  rendererClasses [&NiagaraSpriteRendererProperties]
  renderers [
   {rendererClass=&NiagaraSpriteRendererProperties}
   {rendererClass=&NiagaraSpriteRendererProperties, rendererIndex=1}
  ]
 }
 NE_VC_Boost_Smoke001 {
  bEnabled=false
  emitterUpdateScript [
   &EmitterState()
   &VC_FilteredSocketCount()
   &SpawnRate(
    SpawnRate=&Multiply_Float(
     A=&MakeCustomFloatFromBool(Bool=$User.bIsFrontendPreview, True Float=50, False Float=100)
     B=&VC_Mask_Floats(Coro Value:NiagaraFloat=1, FrontEnd Value:NiagaraFloat=1, FN Value:NiagaraFloat=0.5)
    )
    Use Spawn Probability=false
   )
   &SpawnRate(
    SpawnGroup=2
    SpawnRate=&Multiply_Float(
     A=&MakeCustomFloatFromBool(Bool=$User.bIsFrontendPreview, True Float=50, False Float=200)
     B=&VC_Mask_Floats(Coro Value:NiagaraFloat=1, FrontEnd Value:NiagaraFloat=1, FN Value:NiagaraFloat=0.5)
    )
    Use Spawn Probability=false
   ) {module=SpawnRate002}
   &SpawnRate(
    SpawnGroup=1
    SpawnRate=&Multiply_Float(
     A=50
     B=&VC_Mask_Floats(Coro Value:NiagaraFloat=1, FrontEnd Value:NiagaraFloat=0.5, FN Value:NiagaraFloat=0.5)
    )
    Use Spawn Probability=false
   ) {module=SpawnRate001}
   /SetVariables(Emitter.Detail Power:NiagaraFloat=&MakeCustomFloatFromBool(Bool=$User.bIsFrontendPreview, True Float=0.5, False Float=0)) {isSetParameters=true, module=SetVariables_#4}
  ]
  particleSpawnScript [
   &VC_PaintedVariantsOverride() {enabled=false}
   &InitializeParticle(
    Color=&MergeHelper_ColorSwitch(
     0=&MakeCustomLinearColorFromBool(Bool=$User.bIsFrontendPreview, True Color=<rgba 1 1 1 1>, False Color=<rgba 2 0.4893 0.24 1>)
     1=<rgba 0.1 0 0.0052 1>
     2=<rgba 1 1 1 2>
     3=?
     Index=&SpawnGroup()
     Num=::NewEnumerator1
    )
    Lifetime=&MergeHelper_FloatSwitch(
     0=&MakeCustomFloatFromBool(Bool=$User.bIsFrontendPreview, True Float=1, False Float=1)
     1=&MakeCustomFloatFromBool(Bool=$User.bIsFrontendPreview, True Float=1.5, False Float=1.3)
     2=&MakeCustomFloatFromBool(Bool=$User.bIsFrontendPreview, True Float=&RandomRangeFloat(Minimum=0.45, Maximum=0.3), False Float=0)
     3=?
     Index=&SpawnGroup()
     Num=::NewEnumerator1
    )
    Sprite Rotation Mode=::NewEnumerator2
    Sprite Size Mode=::NewEnumerator3
    Uniform Sprite Size=&MergeHelper_FloatSwitch(
     0=&MakeCustomFloatFromBool(
      Bool=$User.bIsFrontendPreview
      True Float=&RandomRangeFloat(Minimum=500, Maximum=400)
      False Float=&RandomRangeFloat(Minimum=400, Maximum=350)
     )
     1=&MakeCustomFloatFromBool(
      Bool=$User.bIsFrontendPreview
      True Float=&RandomRangeFloat(Minimum=700, Maximum=650)
      False Float=&RandomRangeFloat(Minimum=400, Maximum=550)
     )
     2=&MakeCustomFloatFromBool(
      Bool=$User.bIsFrontendPreview
      True Float=&RandomRangeFloat(Minimum=250, Maximum=200)
      False Float=&RandomRangeFloat(Minimum=400, Maximum=350)
     )
     3=?
     Index=&SpawnGroup()
     Num=::NewEnumerator1
    )
    UsePositionOffset=false
    Write Material Random=true
   )
   &PartitionParticles(Number Of Partitions=$Emitter.PortCount)
   &SkeletalMeshLocation(
    Bone / Socket Index=$Particles.PartitionParticles.Partition
    Bone / Socket Sampling Mode=::NewEnumerator5
    MaskSpawnGroup=false
    Mesh Sampling Type=::NewEnumerator6
    Sampled Position Offset=&MakeCustomPositionFromBool(Bool:NiagaraBool=$User.bIsFrontendPreview, True Position:NiagaraPosition=<0 0 -10>, False Position:NiagaraPosition=<0 0 25>)
    Skeletal Mesh=$System.SkeletalMesh
   )
   &AddVelocity(
    Cone Angle=5
    Cone Axis=&RandomRangeVector(Minimum=<1 -0.12 -0.02>, Maximum=<1 0.12 -0.02>)
    Module.UseConeFalloff=false
    Module.UseVelDistribution=false
    Override Seed=false
    Velocity Mode=::NewEnumerator2
    Velocity Speed=&MakeFloatFromVector(
     Vector=&VC_Mask_Vectors(Coro Value:Vector3f=<-400 0 0>, FrontEnd Value:Vector3f=<-1200 0 0>, FN Value:Vector3f=<-1200 0 0>)
     Channel=::NewEnumerator0
    )
   )
   &ShapeLocation(
    Box Size=&MergeHelper_VectorSwitch(0=<1 80 1>, 1=<1 100 1>, 2=<1 60 1>, 3=?, Index=&SpawnGroup(), Num=::NewEnumerator1)
    MaskSpawnGroup=false
    Override Seed=false
    Shape Primitive=::NewEnumerator2
    Surface Only=false
    UseOwnerScale=true
   )
   &CameraOffset(Camera Offset Amount=&MergeHelper_FloatSwitch(0=50, 1=-5, 2=100, 3=?, Index=&SpawnGroup(), Num=::NewEnumerator1))
   /SetVariables(Particles.VisibilityTag=&SelectIntFromArray(Direct Array Index=&SpawnGroup(), Int Selection Array=DataInterface() {propertyValues {IntData [0,1,2]}})) {isSetParameters=true, module=SetVariables_#5}
   &DLR_ProductPreviews(GameModeID:NiagaraInt32=$System.GameModeID, bPreviewVehicleSpeed:NiagaraBool=$System.bPreviewVehicleSpeed)
  ]
  particleUpdateScript [
   &ParticleState()
   &ScaleSpriteSize(
    Scale Sprite Size Mode=::NewEnumerator0
    Uniform Scale Factor=&MakeCustomFloatFromBool(
     Bool=$User.bIsFrontendPreview
     True Float=&FloatFromCurve(
      FloatCurve=DataInterface() {propertyValues {Curve {
         defaultValue=1
         keys [{arriveTangent=1.0152, interpMode=RCIM_Cubic, leaveTangent=1.0152, tangentMode=RCTM_User, time=0.2, value=0.6}, <at 1 2 RCIM_Cubic>]
        }}}
      Scale Curve=0.8
     )
     False Float=&FloatFromCurve(FloatCurve=DataInterface() {propertyValues {Curve {
         defaultValue=1
         keys [
          {arriveTangent=-0.1605, interpMode=RCIM_Cubic, leaveTangent=-0.1605, tangentMode=RCTM_User, time=0.1, value=0.4}
          {arriveTangent=2.1333, interpMode=RCIM_Cubic, leaveTangent=2.1333, time=0.4, value=0.715}
          {arriveTangent=-0.3583, interpMode=RCIM_Cubic, leaveTangent=-0.3583, time=0.85, value=2}
          <at 1 0.5 RCIM_Cubic>
         ]
        }}})
    )
   )
   &SpriteRotationRate(Rotation Rate=&RandomRangeFloat(Minimum=125, Maximum=-125))
   &DynamicMaterialParameters(
    Index 0 Param 1=1
    Index 0 Param 2=&MergeHelper_FloatSwitch(
     0=&FloatFromCurve(FloatCurve=DataInterface() {propertyValues {Curve {defaultValue=0, keys [{tangentMode=RCTM_User, time=0, value=0}, {tangentMode=RCTM_User, time=1, value=0}]}}})
     1=&FloatFromCurve(FloatCurve=DataInterface() {propertyValues {Curve {defaultValue=1, keys [{tangentMode=RCTM_User, time=0, value=0}, {tangentMode=RCTM_User, time=1, value=0.5}]}}})
     2=?
     3=?
     Index=&SpawnGroup()
     Num=::NewEnumerator0
    )
    Index 0 Param 3=&MergeHelper_FloatSwitch(0=1, 1=0, 2=1, 3=?, Index=&SpawnGroup(), Num=::NewEnumerator1)
    Index 0 Param 4=&MergeHelper_FloatSwitch(
     0=&FloatFromCurve(
      FloatCurve=DataInterface() {propertyValues {Curve {defaultValue=1, keys [{tangentMode=RCTM_User, time=0, value=0}, {tangentMode=RCTM_User, time=1, value=1}]}}}
      Scale Curve=0.05
     )
     1=0
     2=&FloatFromCurve(
      FloatCurve=DataInterface() {propertyValues {Curve {defaultValue=1, keys [{tangentMode=RCTM_User, time=0, value=0}, {tangentMode=RCTM_User, time=1, value=1}]}}}
      Scale Curve=0.1
     )
     3=?
     Index=&SpawnGroup()
     Num=::NewEnumerator1
    )
    Index 2 Param 1=0.5
    Index 2 Param 2=0.5
    Index 2 Param 3=&FloatFromCurve(FloatCurve=DataInterface() {propertyValues {Curve {keys [<at 0 0.5>,<at 1 1>]}}}, Scale Curve=0.5)
    Index 2 Param 4=&FloatFromCurve(FloatCurve=DataInterface() {propertyValues {Curve {keys [<at 0 0.5>,<at 1 1>]}}}, Scale Curve=0.5)
    Module.Index0Float0Write=true
    Module.Index0Float1Write=false
    Module.Index0Float2Write=true
    Module.Index0Float3Write=true
    Module.Index2Float0Write=true
    Module.Index2Float1Write=true
    Module.Index2Float2Write=true
    Module.Index2Float3Write=true
    Write Parameter Index 2=::NewEnumerator1
   )
   &InheritSourceMovement(
    Applied Position Scale=&VectorFromFloat(Value=&FloatFromCurve(FloatCurve=DataInterface() {propertyValues {Curve {
         defaultValue=1
         keys [
          {arriveTangent=-1.7885, interpMode=RCIM_Cubic, leaveTangent=-1.7885, tangentMode=RCTM_User, time=0.05, value=1}
          {interpMode=RCIM_Cubic, tangentMode=RCTM_User, time=1, value=0}
         ]
        }}}))
    ApplyForce=false
    ApplyPosition=true
    ApplyPositionFalloff=false
   )
   &GravityForce(Gravity=<0 0 0>)
   &SolveForcesAndVelocity(Clamp Velocity=false, Limit Acceleration=false)
   &ScaleColor(
    Color Scale:LinearColor=&MergeHelper_ColorSwitch(
     0=&MakeCustomLinearColorFromBool(
      Bool=$User.bIsFrontendPreview
      True Color=&ColorFromCurve(ColorCurve=DataInterface() {propertyValues {
         AlphaCurve {keys [
           <at 0.0565 0 RCIM_Cubic>
           {arriveTangent=0.9091, interpMode=RCIM_Cubic, leaveTangent=0.9091, tangentMode=RCTM_User, time=0.4, value=1}
           <at 0.9346 0.8359>
          ]}
         BlueCurve {keys [<at 0.1625 0.15>,<at 0.7871 0.02>,<at 1 0.02>]}
         GreenCurve {keys [<at 0.1625 0.525>,<at 0.7871 0.0485>,<at 1 0.0341>]}
         RedCurve {keys [
           {arriveTangent=-1.2682, interpMode=RCIM_Cubic, leaveTangent=-1.2682, tangentMode=RCTM_User, time=0, value=2}
           <at 0.1625 3>
           <at 0.7871 0.4>
           <at 1 0.2>
          ]}
        }})
      False Color=&MakeLinearColorFromVectorAndFloat(
       "Vector (RGB)"=&VectorFromCurve(VectorCurve=DataInterface() {propertyValues {
          XCurve {keys [
            {arriveTangent=-5.8345, interpMode=RCIM_Cubic, leaveTangent=-5.8345, tangentMode=RCTM_User, time=0, value=2.1}
            {arriveTangent=-2.08, interpMode=RCIM_Cubic, leaveTangent=-2.08, time=0.12, value=1}
            {arriveTangent=0.068, interpMode=RCIM_Cubic, leaveTangent=0.068, tangentMode=RCTM_User, time=1, value=0.02}
           ]}
          YCurve {keys [
            {arriveTangent=-2.7845, interpMode=RCIM_Cubic, leaveTangent=-2.7845, tangentMode=RCTM_User, time=0, value=1}
            {arriveTangent=-1.0667, interpMode=RCIM_Cubic, leaveTangent=-1.0667, time=0.15, value=0.6}
            {arriveTangent=0.4006, interpMode=RCIM_Cubic, leaveTangent=0.4006, tangentMode=RCTM_User, time=0.9, value=0.04}
           ]}
          ZCurve {keys [
            {arriveTangent=-2.8548, interpMode=RCIM_Cubic, leaveTangent=-2.8548, tangentMode=RCTM_User, time=0, value=1.1}
            {arriveTangent=-0.0519, interpMode=RCIM_Cubic, leaveTangent=-0.0519, tangentMode=RCTM_User, time=0.9, value=0.1}
           ]}
         }})
       "Float (Alpha)"=&FloatFromCurve(FloatCurve=DataInterface() {propertyValues {Curve {keys [<at 0 0 RCIM_Cubic>, {arriveTangent=32.8272, tangentMode=RCTM_User, time=0.15, value=2}, <at 0.8031 1>, <at 1 0.5>]}}})
      )
     )
     1=&MakeCustomLinearColorFromBool(
      Bool=$User.bIsFrontendPreview
      True Color=&ColorFromCurve(ColorCurve=DataInterface() {propertyValues {
         AlphaCurve {keys [<at 0.4412 0 RCIM_Cubic>,<at 0.6991 2 RCIM_Cubic>,<at 0.9902 0>]}
         BlueCurve {keys [<at 0.7289 0 RCIM_Cubic>,<at 0.994 0>]}
         GreenCurve {keys [<at 0.7289 0.0015 RCIM_Cubic>,<at 0.994 0.0006>]}
         RedCurve {keys [<at 0.7289 0.0625 RCIM_Cubic>,<at 0.994 0.026>]}
        }})
      False Color=&ColorFromCurve(ColorCurve=DataInterface() {propertyValues {
         AlphaCurve {keys [<at 0.1348 0 RCIM_Cubic>,<at 0.5405 2 RCIM_Cubic>,<at 0.9902 0>]}
         BlueCurve {keys [<at 0.7289 0 RCIM_Cubic>,<at 0.994 0>]}
         GreenCurve {keys [<at 0.7289 0.0015 RCIM_Cubic>,<at 0.994 0.0006>]}
         RedCurve {keys [<at 0.7289 0.0625 RCIM_Cubic>,<at 0.994 0.026>]}
        }})
     )
     2=&MakeCustomLinearColorFromBool(
      Bool=$User.bIsFrontendPreview
      True Color=&ColorFromCurve(ColorCurve=DataInterface() {propertyValues {
         AlphaCurve {keys [
           <at 0.1604 0 RCIM_Cubic>
           {arriveTangent=0.9091, interpMode=RCIM_Cubic, leaveTangent=0.9091, tangentMode=RCTM_User, time=0.4944, value=1}
           <at 0.8055 0>
          ]}
         BlueCurve {keys [{arriveTangent=-0.32, interpMode=RCIM_Cubic, leaveTangent=-0.32, tangentMode=RCTM_User, time=0.1, value=0.3}, <at 0.2132 0.3>, <at 0.8885 0>]}
         GreenCurve {keys [<at 0.2132 0.75>, {arriveTangent=-1.4613, interpMode=RCIM_Cubic, leaveTangent=-1.4613, tangentMode=RCTM_User, time=0.8885, value=0}]}
         RedCurve {keys [<at 0.2132 3>, {arriveTangent=-12.0185, interpMode=RCIM_Cubic, leaveTangent=-12.0185, tangentMode=RCTM_User, time=0.8885, value=0}]}
        }})
      False Color=&MakeLinearColorFromVectorAndFloat(
       "Vector (RGB)"=&VectorFromCurve(VectorCurve=DataInterface() {propertyValues {
          XCurve {keys [<at 0 1.1>,<at 0.3888 1>,<at 0.9 0.4>]}
          YCurve {keys [
            {arriveTangent=-0.3418, interpMode=RCIM_Cubic, leaveTangent=-0.3418, tangentMode=RCTM_User, time=0, value=1.1}
            {arriveTangent=-0.7778, interpMode=RCIM_Cubic, leaveTangent=-0.7778, time=0.5149, value=0.6753}
            {arriveTangent=-1.5559, interpMode=RCIM_Cubic, leaveTangent=-1.5559, tangentMode=RCTM_User, time=0.9, value=0.4}
           ]}
          ZCurve {keys [<at 0 1.1>,<at 0.5149 0.7349>,<at 0.9 0.4>]}
         }})
       "Float (Alpha)"=&FloatFromCurve(FloatCurve=DataInterface() {propertyValues {Curve {keys [<at 0.1 0>,<at 0.15 1>,<at 1 0>]}}})
      )
     )
     3=?
     Index=&SpawnGroup()
     Num=::NewEnumerator1
    )
    Scale Mode=3
   )
   &ScaleColor(
    Scale RGB=&MergeHelper_VectorSwitch(
     0=&MakeCustomVectorFromBool(
      Bool=$User.bIsFrontendPreview
      True Vector=&VectorFromCurve(VectorCurve=DataInterface() {propertyValues {
         XCurve {keys [<at 0.1 3 RCIM_Cubic>,<at 1 0.5 RCIM_Cubic>]}
         YCurve {keys [<at 0.1 3 RCIM_Cubic>,<at 1 0.5 RCIM_Cubic>]}
         ZCurve {keys [<at 0.1 3 RCIM_Cubic>,<at 1 0.5 RCIM_Cubic>]}
        }})
      False Vector=&VectorFromCurve(VectorCurve=DataInterface() {propertyValues {
         XCurve {keys [<at 0 1>,<at 0.2 3.5>,<at 0.45 0.8>,<at 1 0.5>]}
         YCurve {keys [<at 0 1>,<at 0.2 1.5>,<at 0.3 1.5>,<at 0.55 1>,<at 1 0.5>]}
         ZCurve {keys [<at 0 1>,<at 0.2 1.5>,<at 0.3 1.5>,<at 0.55 1>,<at 1 0.5>]}
        }})
     )
     1=&VectorFromCurve(VectorCurve=DataInterface() {propertyValues {XCurve {keys [<at 0 0.3>,<at 1 0.1>]}, YCurve {keys [<at 0 0.3>,<at 1 0.1>]}, ZCurve {keys [<at 0 0.3>,<at 1 0.1>]}}})
     2=?
     3=?
     Index=&SpawnGroup()
     Num=::NewEnumerator0
    )
    ScaleA=false
    ScaleRGB=true
   ) {enabled=false, module=ScaleColor001}
  ]
  rendererClasses [&NiagaraSpriteRendererProperties]
  renderers [
   {rendererClass=&NiagaraSpriteRendererProperties}
   {rendererClass=&NiagaraSpriteRendererProperties, rendererIndex=1}
   {rendererClass=&NiagaraSpriteRendererProperties, rendererIndex=2}
  ]
 }
 NE_VC_Boost_Smoke002 {
  bEnabled=false
  emitterUpdateScript [
   &EmitterState()
   &VC_FilteredSocketCount()
   &SpawnRate(
    SpawnRate=&Multiply_Float(
     A=30
     B=&VC_Mask_Floats(Coro Value:NiagaraFloat=1, FrontEnd Value:NiagaraFloat=1, FN Value:NiagaraFloat=0.5)
    )
    Use Spawn Probability=false
   )
   &SpawnRate(
    SpawnGroup=1
    SpawnRate=&Multiply_Float(
     A=200
     B=&VC_Mask_Floats(Coro Value:NiagaraFloat=1, FrontEnd Value:NiagaraFloat=0.5, FN Value:NiagaraFloat=0.5)
    )
    Use Spawn Probability=false
   ) {module=SpawnRate001}
   /SetVariables(Emitter.Detail Power:NiagaraFloat=&MakeCustomFloatFromBool(Bool=$User.bIsFrontendPreview, True Float=1, False Float=0.5)) {isSetParameters=true, module=SetVariables_#4}
  ]
  particleSpawnScript [
   &VC_PaintedVariantsOverride() {enabled=false}
   &InitializeParticle(
    Color=&MergeHelper_ColorSwitch(
     0=&MakeCustomLinearColorFromBool(Bool=$User.bIsFrontendPreview, True Color=<rgba 1 1 1 1>, False Color=<rgba 3 0.489 0.21 10>)
     1=<rgba 0.1 0.0191 0.01 1>
     2=?
     3=?
     Index=&SpawnGroup()
     Num=::NewEnumerator0
    )
    Lifetime=&MergeHelper_FloatSwitch(
     0=&MakeCustomFloatFromBool(Bool=$User.bIsFrontendPreview, True Float=0.5, False Float=1.5)
     1=&MakeCustomFloatFromBool(Bool=$User.bIsFrontendPreview, True Float=0.88, False Float=1.05)
     2=&RandomRangeFloat(Minimum=0.1, Maximum=0.4)
     3=?
     Index=&SpawnGroup()
     Num=::NewEnumerator0
    )
    Sprite Rotation Mode=::NewEnumerator2
    Sprite Size Mode=::NewEnumerator3
    Uniform Sprite Size=&MergeHelper_FloatSwitch(
     0=&MakeCustomFloatFromBool(
      Bool=$User.bIsFrontendPreview
      True Float=&RandomRangeFloat(Minimum=400, Maximum=300)
      False Float=&RandomRangeFloat(Minimum=400, Maximum=350)
     )
     1=&MakeCustomFloatFromBool(
      Bool=$User.bIsFrontendPreview
      True Float=&RandomRangeFloat(Minimum=500, Maximum=600)
      False Float=&RandomRangeFloat(Minimum=600, Maximum=700)
     )
     2=?
     3=?
     Index=&SpawnGroup()
     Num=::NewEnumerator0
    )
    UsePositionOffset=false
    Write Material Random=true
   )
   &PartitionParticles(Number Of Partitions=$Emitter.PortCount)
   &SkeletalMeshLocation(
    Bone / Socket Index=$Particles.PartitionParticles.Partition
    Bone / Socket Sampling Mode=::NewEnumerator5
    MaskSpawnGroup=false
    Mesh Sampling Type=::NewEnumerator6
    Sampled Position Offset=&MakeCustomPositionFromBool(Bool:NiagaraBool=$User.bIsFrontendPreview, True Position:NiagaraPosition=<0 0 -10>, False Position:NiagaraPosition=<0 0 25>)
    Skeletal Mesh=$System.SkeletalMesh
   )
   &AddVelocity(
    Cone Angle=5
    Cone Axis=&RandomRangeVector(Minimum=<1 -0.12 -0.02>, Maximum=<1 0.12 -0.02>)
    Module.UseConeFalloff=false
    Module.UseVelDistribution=false
    Override Seed=false
    Velocity Mode=::NewEnumerator2
    Velocity Speed=&MakeFloatFromVector(
     Vector=&VC_Mask_Vectors(Coro Value:Vector3f=<-400 0 0>, FrontEnd Value:Vector3f=<-1200 0 0>, FN Value:Vector3f=<-1200 0 0>)
     Channel=::NewEnumerator0
    )
   )
   &ShapeLocation(
    Box Size=&MergeHelper_VectorSwitch(0=<1 80 1>, 1=<1 100 1>, 2=?, 3=?, Index=&SpawnGroup(), Num=::NewEnumerator0)
    MaskSpawnGroup=false
    Override Seed=false
    Shape Primitive=::NewEnumerator2
    Surface Only=false
    UseOwnerScale=true
   )
   &CameraOffset(Camera Offset Amount=&MergeHelper_FloatSwitch(0=80, 1=-5, 2=?, 3=?, Index=&SpawnGroup(), Num=::NewEnumerator0))
   /SetVariables(Particles.VisibilityTag=$Emitter.SpawnGroup) {isSetParameters=true, module=SetVariables_#5}
   &DLR_ProductPreviews(GameModeID:NiagaraInt32=$System.GameModeID, bPreviewVehicleSpeed:NiagaraBool=$System.bPreviewVehicleSpeed)
  ]
  particleUpdateScript [
   &ParticleState()
   &ScaleSpriteSize(
    Scale Sprite Size Mode=::NewEnumerator0
    Uniform Scale Factor=&MakeCustomFloatFromBool(
     Bool=$User.bIsFrontendPreview
     True Float=&FloatFromCurve(FloatCurve=DataInterface() {propertyValues {Curve {
         defaultValue=1
         keys [{arriveTangent=0.644, interpMode=RCIM_Cubic, leaveTangent=0.644, tangentMode=RCTM_User, time=0.1, value=0.6}, <at 1 1 RCIM_Cubic>]
        }}})
     False Float=&FloatFromCurve(FloatCurve=DataInterface() {propertyValues {Curve {
         defaultValue=1
         keys [
          {arriveTangent=-0.1605, interpMode=RCIM_Cubic, leaveTangent=-0.1605, tangentMode=RCTM_User, time=0.1, value=0.6}
          {arriveTangent=0.6443, interpMode=RCIM_Cubic, leaveTangent=0.6443, time=0.4763, value=0.715}
          {arriveTangent=0.7351, interpMode=RCIM_Cubic, leaveTangent=0.7351, time=0.6802, value=0.9738}
          <at 1 1.1 RCIM_Cubic>
         ]
        }}})
    )
   )
   &SpriteRotationRate(Rotation Rate=&RandomRangeFloat(Minimum=125, Maximum=-125))
   &DynamicMaterialParameters(
    Index 0 Param 1=1
    Index 0 Param 2=&MergeHelper_FloatSwitch(
     0=&FloatFromCurve(FloatCurve=DataInterface() {propertyValues {Curve {defaultValue=0, keys [{tangentMode=RCTM_User, time=0, value=0}, {tangentMode=RCTM_User, time=1, value=0}]}}})
     1=&FloatFromCurve(FloatCurve=DataInterface() {propertyValues {Curve {defaultValue=1, keys [{tangentMode=RCTM_User, time=0, value=0}, {tangentMode=RCTM_User, time=1, value=0.5}]}}})
     2=?
     3=?
     Index=&SpawnGroup()
     Num=::NewEnumerator0
    )
    Index 0 Param 3=&MergeHelper_FloatSwitch(0=0.2, 1=0, 2=?, 3=?, Index=&SpawnGroup(), Num=::NewEnumerator0)
    Index 0 Param 4=&MergeHelper_FloatSwitch(
     0=&FloatFromCurve(
      FloatCurve=DataInterface() {propertyValues {Curve {defaultValue=1, keys [{tangentMode=RCTM_User, time=0, value=0}, {tangentMode=RCTM_User, time=1, value=1}]}}}
      Scale Curve=0.1
     )
     1=0
     2=?
     3=?
     Index=&SpawnGroup()
     Num=::NewEnumerator0
    )
    Index 2 Param 1=0.5
    Index 2 Param 2=0.5
    Index 2 Param 3=&FloatFromCurve(FloatCurve=DataInterface() {propertyValues {Curve {keys [<at 0 0.5>,<at 1 1>]}}}, Scale Curve=0.5)
    Index 2 Param 4=&FloatFromCurve(FloatCurve=DataInterface() {propertyValues {Curve {keys [<at 0 0.5>,<at 1 1>]}}}, Scale Curve=0.5)
    Module.Index0Float0Write=true
    Module.Index0Float1Write=false
    Module.Index0Float2Write=true
    Module.Index0Float3Write=true
    Module.Index2Float0Write=true
    Module.Index2Float1Write=true
    Module.Index2Float2Write=true
    Module.Index2Float3Write=true
    Write Parameter Index 2=::NewEnumerator1
   )
   &InheritSourceMovement(
    Applied Position Scale=&VectorFromFloat(Value=&FloatFromCurve(FloatCurve=DataInterface() {propertyValues {Curve {
         defaultValue=1
         keys [
          {arriveTangent=-1.7885, interpMode=RCIM_Cubic, leaveTangent=-1.7885, tangentMode=RCTM_User, time=0.05, value=1}
          {interpMode=RCIM_Cubic, tangentMode=RCTM_User, time=1, value=0}
         ]
        }}}))
    ApplyForce=false
    ApplyPosition=true
    ApplyPositionFalloff=false
   )
   &SolveForcesAndVelocity(Clamp Velocity=false, Limit Acceleration=false)
   &ScaleColor(
    Color Scale:LinearColor=&MergeHelper_ColorSwitch(
     0=&MakeCustomLinearColorFromBool(
      Bool=$User.bIsFrontendPreview
      True Color=&ColorFromCurve(ColorCurve=DataInterface() {propertyValues {
         AlphaCurve {keys [
           <at 0.0241 0 RCIM_Cubic>
           {arriveTangent=0.9091, interpMode=RCIM_Cubic, leaveTangent=0.9091, tangentMode=RCTM_User, time=0.4586, value=1}
           <at 0.9857 0>
          ]}
         BlueCurve {keys [
           <at 0 1.4>
           {arriveTangent=-0.32, interpMode=RCIM_Cubic, leaveTangent=-0.32, tangentMode=RCTM_User, time=0.1, value=0.3}
           <at 0.4317 0.4572>
           <at 0.813 0.15>
          ]}
         GreenCurve {keys [
           <at 0 6.56>
           <at 0.4317 3.4905>
           {arriveTangent=-1.4613, interpMode=RCIM_Cubic, leaveTangent=-1.4613, tangentMode=RCTM_User, time=0.813, value=0.4385}
          ]}
         RedCurve {keys [
           <at 0 20>
           <at 0.4317 10.9724>
           {arriveTangent=-12.0185, interpMode=RCIM_Cubic, leaveTangent=-12.0185, tangentMode=RCTM_User, time=0.813, value=3}
          ]}
        }})
      False Color=&ColorFromCurve(ColorCurve=DataInterface() {propertyValues {
         AlphaCurve {keys [<at 0.0241 0>,<at 0.25 5 RCIM_Cubic>]}
         BlueCurve {keys [
           {arriveTangent=-0.8352, interpMode=RCIM_Cubic, leaveTangent=-0.8352, tangentMode=RCTM_User, time=0, value=0.283}
           {arriveTangent=-0.2931, interpMode=RCIM_Cubic, leaveTangent=-0.2931, tangentMode=RCTM_User, time=0.4631, value=0.0208}
           {interpMode=RCIM_Cubic, leaveTangent=-0.0388, time=0.6, value=0}
           <at 0.8607 0 RCIM_Cubic>
          ]}
         GreenCurve {keys [
           <at 0 0.8 RCIM_Cubic>
           {arriveTangent=-1.1667, interpMode=RCIM_Cubic, leaveTangent=-1.1667, time=0.3194, value=0.3}
           {interpMode=RCIM_Cubic, leaveTangent=-0.3796, time=0.6, value=0.1}
           <at 0.8607 0.0417 RCIM_Cubic>
          ]}
         RedCurve {keys [
           <at 0 5 RCIM_Cubic>
           {arriveTangent=-10.4931, interpMode=RCIM_Cubic, leaveTangent=-10.4931, tangentMode=RCTM_User, time=0.3, value=3}
           {arriveTangent=-3.6293, interpMode=RCIM_Cubic, leaveTangent=-3.6293, tangentMode=RCTM_User, time=0.55, value=1}
           {arriveTangent=0.4301, interpMode=RCIM_Cubic, leaveTangent=0.4301, tangentMode=RCTM_User, time=0.8607, value=0.5}
          ]}
        }})
     )
     1=&ColorFromCurve(ColorCurve=DataInterface() {propertyValues {
        AlphaCurve {keys [<at 0.5236 0 RCIM_Cubic>,<at 1 2 RCIM_Cubic>]}
        BlueCurve {keys [<at 0.7289 0 RCIM_Cubic>,<at 0.994 0>]}
        GreenCurve {keys [<at 0.7289 0.0015 RCIM_Cubic>,<at 0.994 0.0006>]}
        RedCurve {keys [<at 0.7289 0.0625 RCIM_Cubic>,<at 0.994 0.026>]}
       }})
     2=?
     3=?
     Index=&SpawnGroup()
     Num=::NewEnumerator0
    )
    Scale Mode=3
   )
   &ScaleColor(
    Scale RGB=&MergeHelper_VectorSwitch(
     0=&MakeCustomVectorFromBool(
      Bool=$User.bIsFrontendPreview
      True Vector=&VectorFromCurve(VectorCurve=DataInterface() {propertyValues {
         XCurve {keys [<at 0.1 3 RCIM_Cubic>,<at 1 0.5 RCIM_Cubic>]}
         YCurve {keys [<at 0.1 3 RCIM_Cubic>,<at 1 0.5 RCIM_Cubic>]}
         ZCurve {keys [<at 0.1 3 RCIM_Cubic>,<at 1 0.5 RCIM_Cubic>]}
        }})
      False Vector=&VectorFromCurve(VectorCurve=DataInterface() {propertyValues {
         XCurve {keys [<at 0 1>,<at 0.2 3.5>,<at 0.45 0.8>,<at 1 0.5>]}
         YCurve {keys [<at 0 1>,<at 0.2 1.5>,<at 0.3 1.5>,<at 0.55 1>,<at 1 0.5>]}
         ZCurve {keys [<at 0 1>,<at 0.2 1.5>,<at 0.3 1.5>,<at 0.55 1>,<at 1 0.5>]}
        }})
     )
     1=&VectorFromCurve(VectorCurve=DataInterface() {propertyValues {XCurve {keys [<at 0 0.3>,<at 1 0.1>]}, YCurve {keys [<at 0 0.3>,<at 1 0.1>]}, ZCurve {keys [<at 0 0.3>,<at 1 0.1>]}}})
     2=?
     3=?
     Index=&SpawnGroup()
     Num=::NewEnumerator0
    )
    ScaleA=false
    ScaleRGB=true
   ) {enabled=false, module=ScaleColor001}
  ]
  rendererClasses [&NiagaraSpriteRendererProperties]
  renderers [
   {rendererClass=&NiagaraSpriteRendererProperties}
   {rendererClass=&NiagaraSpriteRendererProperties, rendererIndex=1}
  ]
 }
 NE_VC_Boost_Smoke003 {
  bEnabled=false
  emitterUpdateScript [
   &EmitterState()
   &VC_FilteredSocketCount()
   &SpawnRate(
    SpawnRate=&Multiply_Float(
     A=150
     B=&VC_Mask_Floats(Coro Value:NiagaraFloat=1, FrontEnd Value:NiagaraFloat=0.6, FN Value:NiagaraFloat=0.5)
    )
    Use Spawn Probability=false
   )
   &SpawnRate(
    SpawnGroup=2
    SpawnRate=&Multiply_Float(
     A=40
     B=&VC_Mask_Floats(Coro Value:NiagaraFloat=1, FrontEnd Value:NiagaraFloat=1, FN Value:NiagaraFloat=0.5)
    )
    Use Spawn Probability=false
   ) {module=SpawnRate002}
   &SpawnRate(
    SpawnGroup=1
    SpawnRate=&Multiply_Float(
     A=200
     B=&VC_Mask_Floats(Coro Value:NiagaraFloat=1, FrontEnd Value:NiagaraFloat=0.5, FN Value:NiagaraFloat=0.5)
    )
    Use Spawn Probability=false
   ) {module=SpawnRate001}
   /SetVariables(Emitter.Detail Power:NiagaraFloat=&MakeCustomFloatFromBool(Bool=$User.bIsFrontendPreview, True Float=1, False Float=0.3)) {isSetParameters=true, module=SetVariables_#4}
  ]
  particleSpawnScript [
   &VC_PaintedVariantsOverride() {enabled=false}
   &InitializeParticle(
    Color=&MergeHelper_ColorSwitch(0=<rgba 1 1 1 2>, 1=<rgba 1 1 1 1>, 2=<rgba 1 1 1 2>, 3=?, Index=&SpawnGroup(), Num=::NewEnumerator1)
    Lifetime=&MergeHelper_FloatSwitch(
     0=&MakeCustomFloatFromBool(Bool=$User.bIsFrontendPreview, True Float=0.9, False Float=1)
     1=&MakeCustomFloatFromBool(Bool=$User.bIsFrontendPreview, True Float=0.9, False Float=1)
     2=&MakeCustomFloatFromBool(Bool=$User.bIsFrontendPreview, True Float=&RandomRangeFloat(Minimum=0.2, Maximum=0.5), False Float=0)
     3=?
     Index=&SpawnGroup()
     Num=::NewEnumerator1
    )
    Sprite Rotation Mode=::NewEnumerator2
    Sprite Size Mode=::NewEnumerator3
    Uniform Sprite Size=&MergeHelper_FloatSwitch(
     0=&MakeCustomFloatFromBool(
      Bool=$User.bIsFrontendPreview
      True Float=&RandomRangeFloat(Minimum=300, Maximum=250)
      False Float=&RandomRangeFloat(Minimum=400, Maximum=350)
     )
     1=&MakeCustomFloatFromBool(
      Bool=$User.bIsFrontendPreview
      True Float=&RandomRangeFloat(Minimum=350, Maximum=400)
      False Float=&RandomRangeFloat(Minimum=400, Maximum=550)
     )
     2=&MakeCustomFloatFromBool(
      Bool=$User.bIsFrontendPreview
      True Float=&RandomRangeFloat(Minimum=150, Maximum=200)
      False Float=&RandomRangeFloat(Minimum=400, Maximum=350)
     )
     3=?
     Index=&SpawnGroup()
     Num=::NewEnumerator1
    )
    UsePositionOffset=false
    Write Material Random=true
   )
   &PartitionParticles(Number Of Partitions=$Emitter.PortCount)
   &SkeletalMeshLocation(
    Bone / Socket Index=$Particles.PartitionParticles.Partition
    Bone / Socket Sampling Mode=::NewEnumerator5
    MaskSpawnGroup=false
    Mesh Sampling Type=::NewEnumerator6
    Sampled Position Offset=&MakeCustomPositionFromBool(Bool:NiagaraBool=$User.bIsFrontendPreview, True Position:NiagaraPosition=<0 0 -10>, False Position:NiagaraPosition=<0 0 25>)
    Skeletal Mesh=$System.SkeletalMesh
   )
   &AddVelocity(
    Cone Angle=5
    Cone Axis=&RandomRangeVector(Minimum=<1 -0.12 -0.02>, Maximum=<1 0.12 -0.02>)
    Module.UseConeFalloff=false
    Module.UseVelDistribution=false
    Override Seed=false
    Velocity Mode=::NewEnumerator2
    Velocity Speed=&MakeFloatFromVector(
     Vector=&VC_Mask_Vectors(Coro Value:Vector3f=<-400 0 0>, FrontEnd Value:Vector3f=<-1200 0 0>, FN Value:Vector3f=<-1200 0 0>)
     Channel=::NewEnumerator0
    )
   )
   &ShapeLocation(
    Box Size=&MergeHelper_VectorSwitch(0=<1 80 1>, 1=<1 100 1>, 2=<1 60 1>, 3=?, Index=&SpawnGroup(), Num=::NewEnumerator1)
    MaskSpawnGroup=false
    Override Seed=false
    Shape Primitive=::NewEnumerator2
    Surface Only=false
    UseOwnerScale=true
   )
   &CameraOffset(Camera Offset Amount=&MergeHelper_FloatSwitch(0=50, 1=-50, 2=100, 3=?, Index=&SpawnGroup(), Num=::NewEnumerator1))
   /SetVariables(Particles.VisibilityTag=&SelectIntFromArray(Direct Array Index=&SpawnGroup(), Int Selection Array=DataInterface() {propertyValues {IntData [0,1,2]}})) {isSetParameters=true, module=SetVariables_#5}
   &DLR_ProductPreviews(GameModeID:NiagaraInt32=$System.GameModeID, bPreviewVehicleSpeed:NiagaraBool=$System.bPreviewVehicleSpeed)
  ]
  particleUpdateScript [
   &ParticleState()
   &ScaleSpriteSize(
    Scale Sprite Size Mode=::NewEnumerator0
    Uniform Scale Factor=&MakeCustomFloatFromBool(
     Bool=$User.bIsFrontendPreview
     True Float=&FloatFromCurve(FloatCurve=DataInterface() {propertyValues {Curve {
         defaultValue=1
         keys [{arriveTangent=0.4477, interpMode=RCIM_Cubic, leaveTangent=0.4477, tangentMode=RCTM_User, time=0.1, value=0.8}, <at 1 1 RCIM_Cubic>]
        }}})
     False Float=&FloatFromCurve(FloatCurve=DataInterface() {propertyValues {Curve {
         defaultValue=1
         keys [
          {arriveTangent=-0.1605, interpMode=RCIM_Cubic, leaveTangent=-0.1605, tangentMode=RCTM_User, time=0.1, value=0.5}
          {arriveTangent=0.8167, interpMode=RCIM_Cubic, leaveTangent=0.8167, time=0.4763, value=0.715}
          {arriveTangent=0.7351, interpMode=RCIM_Cubic, leaveTangent=0.7351, time=0.6802, value=0.9738}
          <at 1 1.1 RCIM_Cubic>
         ]
        }}})
    )
   )
   &SpriteRotationRate(Rotation Rate=&RandomRangeFloat(Minimum=125, Maximum=-125))
   &DynamicMaterialParameters(
    Index 0 Param 1=1
    Index 0 Param 2=&MergeHelper_FloatSwitch(
     0=&FloatFromCurve(FloatCurve=DataInterface() {propertyValues {Curve {defaultValue=0, keys [{tangentMode=RCTM_User, time=0, value=0}, {tangentMode=RCTM_User, time=1, value=0}]}}})
     1=&FloatFromCurve(FloatCurve=DataInterface() {propertyValues {Curve {defaultValue=1, keys [{tangentMode=RCTM_User, time=0, value=0}, {tangentMode=RCTM_User, time=1, value=0.5}]}}})
     2=?
     3=?
     Index=&SpawnGroup()
     Num=::NewEnumerator0
    )
    Index 0 Param 3=&MergeHelper_FloatSwitch(0=0.3, 1=0, 2=0.3, 3=?, Index=&SpawnGroup(), Num=::NewEnumerator1)
    Index 0 Param 4=&MergeHelper_FloatSwitch(
     0=&FloatFromCurve(
      FloatCurve=DataInterface() {propertyValues {Curve {defaultValue=1, keys [{tangentMode=RCTM_User, time=0, value=0}, {tangentMode=RCTM_User, time=1, value=1}]}}}
      Scale Curve=0.05
     )
     1=0
     2=&FloatFromCurve(
      FloatCurve=DataInterface() {propertyValues {Curve {defaultValue=1, keys [{tangentMode=RCTM_User, time=0, value=0}, {tangentMode=RCTM_User, time=1, value=1}]}}}
      Scale Curve=0.1
     )
     3=?
     Index=&SpawnGroup()
     Num=::NewEnumerator1
    )
    Index 2 Param 1=0.5
    Index 2 Param 2=0.5
    Index 2 Param 3=&FloatFromCurve(FloatCurve=DataInterface() {propertyValues {Curve {keys [<at 0 0.5>,<at 1 1>]}}}, Scale Curve=0.5)
    Index 2 Param 4=&FloatFromCurve(FloatCurve=DataInterface() {propertyValues {Curve {keys [<at 0 0.5>,<at 1 1>]}}}, Scale Curve=0.5)
    Module.Index0Float0Write=true
    Module.Index0Float1Write=false
    Module.Index0Float2Write=true
    Module.Index0Float3Write=true
    Module.Index2Float0Write=true
    Module.Index2Float1Write=true
    Module.Index2Float2Write=true
    Module.Index2Float3Write=true
    Write Parameter Index 2=::NewEnumerator1
   )
   &InheritSourceMovement(
    Applied Position Scale=&VectorFromFloat(Value=&FloatFromCurve(FloatCurve=DataInterface() {propertyValues {Curve {
         defaultValue=1
         keys [
          {arriveTangent=-1.7885, interpMode=RCIM_Cubic, leaveTangent=-1.7885, tangentMode=RCTM_User, time=0.05, value=1}
          {interpMode=RCIM_Cubic, tangentMode=RCTM_User, time=1, value=0}
         ]
        }}}))
    ApplyForce=false
    ApplyPosition=true
    ApplyPositionFalloff=false
   )
   &SolveForcesAndVelocity(Clamp Velocity=false, Limit Acceleration=false)
   &ScaleColor(
    Color Scale:LinearColor=&MergeHelper_ColorSwitch(
     0=&MakeCustomLinearColorFromBool(
      Bool=$User.bIsFrontendPreview
      True Color=&ColorFromCurve(ColorCurve=DataInterface() {propertyValues {
         AlphaCurve {keys [
           <at 0.0565 0 RCIM_Cubic>
           {arriveTangent=0.9091, interpMode=RCIM_Cubic, leaveTangent=0.9091, tangentMode=RCTM_User, time=0.4, value=1}
           <at 0.9251 1>
           <at 1 0>
          ]}
         BlueCurve {keys [<at 0 0.06>,<at 1 0.016>]}
         GreenCurve {keys [<at 0 0.3>,<at 1 0.015>]}
         RedCurve {keys [<at 0 2>, {arriveTangent=-1.2682, interpMode=RCIM_Cubic, leaveTangent=-1.2682, tangentMode=RCTM_User, time=0, value=2}, <at 1 0.32>]}
        }})
      False Color=&ColorFromCurve(ColorCurve=DataInterface() {propertyValues {
         AlphaCurve {keys [
           <at 0 0 RCIM_Cubic>
           {arriveTangent=0.9091, interpMode=RCIM_Cubic, leaveTangent=0.9091, tangentMode=RCTM_User, time=0.2, value=2}
           <at 0.7031 1>
           <at 1 0>
          ]}
         BlueCurve {keys [<at 0 0.5>,<at 0.3689 0.108>,<at 0.7076 0.021>,<at 1 0>]}
         GreenCurve {keys [<at 0 0.95>,<at 0.3689 0.152>,<at 0.7076 0.0536>,<at 1 0.0292>]}
         RedCurve {keys [<at 0 5>,<at 0.3689 0.9>,<at 0.7076 0.3>,<at 1 0.25>]}
        }})
     )
     1=&ColorFromCurve(ColorCurve=DataInterface() {propertyValues {
        AlphaCurve {keys [<at 0.2723 0 RCIM_Cubic>,<at 1 0 RCIM_Cubic>]}
        BlueCurve {keys [<at 0.2284 0.015>, {arriveTangent=-0.0196, interpMode=RCIM_Cubic, leaveTangent=-0.0196, time=0.6246, value=0}, <at 0.994 0>]}
        GreenCurve {keys [<at 0.2284 0.0388>, {arriveTangent=-0.0498, interpMode=RCIM_Cubic, leaveTangent=-0.0498, time=0.6246, value=0.0083}, <at 0.994 0.0006>]}
        RedCurve {keys [<at 0.2284 0.3>, {arriveTangent=-0.3578, interpMode=RCIM_Cubic, leaveTangent=-0.3578, time=0.6246, value=0.1}, <at 0.994 0.026>]}
       }})
     2=&MakeCustomLinearColorFromBool(
      Bool=$User.bIsFrontendPreview
      True Color=&ColorFromCurve(ColorCurve=DataInterface() {propertyValues {
         AlphaCurve {keys [
           <at 0.1604 0 RCIM_Cubic>
           {arriveTangent=0.9091, interpMode=RCIM_Cubic, leaveTangent=0.9091, tangentMode=RCTM_User, time=0.4944, value=2}
           <at 0.8055 0>
          ]}
         BlueCurve {keys [{arriveTangent=-0.32, interpMode=RCIM_Cubic, leaveTangent=-0.32, tangentMode=RCTM_User, time=0.1, value=0.3}, <at 0.2132 0.4>, <at 0.8885 0>]}
         GreenCurve {keys [<at 0.2132 1>, {arriveTangent=-1.4613, interpMode=RCIM_Cubic, leaveTangent=-1.4613, tangentMode=RCTM_User, time=0.8885, value=0}]}
         RedCurve {keys [<at 0.2132 4>, {arriveTangent=-12.0185, interpMode=RCIM_Cubic, leaveTangent=-12.0185, tangentMode=RCTM_User, time=0.8885, value=0}]}
        }})
      False Color=&MakeLinearColorFromVectorAndFloat(
       "Vector (RGB)"=&VectorFromCurve(VectorCurve=DataInterface() {propertyValues {
          XCurve {keys [<at 0 1.1>,<at 0.3888 1>,<at 0.9 0.4>]}
          YCurve {keys [
            {arriveTangent=-0.3418, interpMode=RCIM_Cubic, leaveTangent=-0.3418, tangentMode=RCTM_User, time=0, value=1.1}
            {arriveTangent=-0.7778, interpMode=RCIM_Cubic, leaveTangent=-0.7778, time=0.5149, value=0.6753}
            {arriveTangent=-1.5559, interpMode=RCIM_Cubic, leaveTangent=-1.5559, tangentMode=RCTM_User, time=0.9, value=0.4}
           ]}
          ZCurve {keys [<at 0 1.1>,<at 0.5149 0.7349>,<at 0.9 0.4>]}
         }})
       "Float (Alpha)"=&FloatFromCurve(FloatCurve=DataInterface() {propertyValues {Curve {keys [<at 0.1 0>,<at 0.15 1>,<at 1 0>]}}})
      )
     )
     3=?
     Index=&SpawnGroup()
     Num=::NewEnumerator1
    )
    Scale Mode=3
   )
   &ScaleColor(
    Scale RGB=&MergeHelper_VectorSwitch(
     0=&MakeCustomVectorFromBool(
      Bool=$User.bIsFrontendPreview
      True Vector=&VectorFromCurve(VectorCurve=DataInterface() {propertyValues {
         XCurve {keys [<at 0.1 3 RCIM_Cubic>,<at 1 0.5 RCIM_Cubic>]}
         YCurve {keys [<at 0.1 3 RCIM_Cubic>,<at 1 0.5 RCIM_Cubic>]}
         ZCurve {keys [<at 0.1 3 RCIM_Cubic>,<at 1 0.5 RCIM_Cubic>]}
        }})
      False Vector=&VectorFromCurve(VectorCurve=DataInterface() {propertyValues {
         XCurve {keys [<at 0 1>,<at 0.2 3.5>,<at 0.45 0.8>,<at 1 0.5>]}
         YCurve {keys [<at 0 1>,<at 0.2 1.5>,<at 0.3 1.5>,<at 0.55 1>,<at 1 0.5>]}
         ZCurve {keys [<at 0 1>,<at 0.2 1.5>,<at 0.3 1.5>,<at 0.55 1>,<at 1 0.5>]}
        }})
     )
     1=&VectorFromCurve(VectorCurve=DataInterface() {propertyValues {XCurve {keys [<at 0 0.3>,<at 1 0.1>]}, YCurve {keys [<at 0 0.3>,<at 1 0.1>]}, ZCurve {keys [<at 0 0.3>,<at 1 0.1>]}}})
     2=?
     3=?
     Index=&SpawnGroup()
     Num=::NewEnumerator0
    )
    ScaleA=false
    ScaleRGB=true
   ) {enabled=false, module=ScaleColor001}
  ]
  rendererClasses [&NiagaraSpriteRendererProperties]
  renderers [
   {rendererClass=&NiagaraSpriteRendererProperties}
   {rendererClass=&NiagaraSpriteRendererProperties, rendererIndex=1}
   {rendererClass=&NiagaraSpriteRendererProperties, rendererIndex=2}
  ]
 }
 NE_VC_Boost_Smoke004 {
  bEnabled=false
  emitterUpdateScript [
   &EmitterState()
   &VC_FilteredSocketCount()
   &SpawnRate(
    SpawnRate=&Multiply_Float(
     A=&MakeCustomFloatFromBool(Bool=$User.bIsFrontendPreview, True Float=40, False Float=30)
     B=&VC_Mask_Floats(Coro Value:NiagaraFloat=1, FrontEnd Value:NiagaraFloat=1, FN Value:NiagaraFloat=0.5)
    )
    Use Spawn Probability=false
   )
   &SpawnRate(
    SpawnGroup=2
    SpawnRate=&Multiply_Float(
     A=35
     B=&VC_Mask_Floats(Coro Value:NiagaraFloat=1, FrontEnd Value:NiagaraFloat=1, FN Value:NiagaraFloat=0.5)
    )
    Use Spawn Probability=false
   ) {module=SpawnRate002}
   &SpawnRate(
    SpawnGroup=1
    SpawnRate=&Multiply_Float(
     A=200
     B=&VC_Mask_Floats(Coro Value:NiagaraFloat=1, FrontEnd Value:NiagaraFloat=0.5, FN Value:NiagaraFloat=0.5)
    )
    Use Spawn Probability=false
   ) {module=SpawnRate001}
   /SetVariables(
    Emitter.Detail Power:NiagaraFloat=&MakeCustomFloatFromBool(Bool=$User.bIsFrontendPreview, True Float=3, False Float=0.5)
    Emitter.Detail Power02:NiagaraFloat=&MakeCustomFloatFromBool(Bool=$User.bIsFrontendPreview, True Float=2, False Float=0.5)
   ) {isSetParameters=true, module=SetVariables_#4}
  ]
  particleSpawnScript [
   &VC_PaintedVariantsOverride() {enabled=false}
   &InitializeParticle(
    Color=&MergeHelper_ColorSwitch(
     0=&MakeCustomLinearColorFromBool(Bool=$User.bIsFrontendPreview, True Color=<rgba 1 1 1 2>, False Color=<rgba 1.5 1.5 1.5 5>)
     1=<rgba 0.1 0 0.0052 1>
     2=<rgba 2 2 2 2>
     3=?
     Index=&SpawnGroup()
     Num=::NewEnumerator1
    )
    Lifetime=&MergeHelper_FloatSwitch(
     0=&MakeCustomFloatFromBool(Bool=$User.bIsFrontendPreview, True Float=1, False Float=&RandomRangeFloat(Minimum=1.3, Maximum=0.8))
     1=&MakeCustomFloatFromBool(Bool=$User.bIsFrontendPreview, True Float=0.7, False Float=1)
     2=&MakeCustomFloatFromBool(Bool=$User.bIsFrontendPreview, True Float=&RandomRangeFloat(Minimum=0.8, Maximum=0.3), False Float=0)
     3=?
     Index=&SpawnGroup()
     Num=::NewEnumerator1
    )
    Sprite Rotation Mode=::NewEnumerator2
    Sprite Size Mode=::NewEnumerator3
    Uniform Sprite Size=&MergeHelper_FloatSwitch(
     0=&MakeCustomFloatFromBool(
      Bool=$User.bIsFrontendPreview
      True Float=&RandomRangeFloat(Minimum=350, Maximum=200)
      False Float=&RandomRangeFloat(Minimum=200, Maximum=250)
     )
     1=&MakeCustomFloatFromBool(
      Bool=$User.bIsFrontendPreview
      True Float=&RandomRangeFloat(Minimum=600, Maximum=550)
      False Float=&RandomRangeFloat(Minimum=400, Maximum=550)
     )
     2=&MakeCustomFloatFromBool(
      Bool=$User.bIsFrontendPreview
      True Float=&RandomRangeFloat(Minimum=80, Maximum=100)
      False Float=&RandomRangeFloat(Minimum=400, Maximum=350)
     )
     3=?
     Index=&SpawnGroup()
     Num=::NewEnumerator1
    )
    UsePositionOffset=false
    Write Material Random=true
   )
   &PartitionParticles(Number Of Partitions=$Emitter.PortCount)
   &SkeletalMeshLocation(
    Bone / Socket Index=$Particles.PartitionParticles.Partition
    Bone / Socket Sampling Mode=::NewEnumerator5
    MaskSpawnGroup=false
    Mesh Sampling Type=::NewEnumerator6
    Sampled Position Offset=&MakeCustomPositionFromBool(Bool:NiagaraBool=$User.bIsFrontendPreview, True Position:NiagaraPosition=<0 0 -10>, False Position:NiagaraPosition=<0 0 25>)
    Skeletal Mesh=$System.SkeletalMesh
   )
   &AddVelocity(
    Cone Angle=5
    Cone Axis=&RandomRangeVector(Minimum=<1 -0.12 -0.02>, Maximum=<1 0.12 -0.02>)
    Module.UseConeFalloff=false
    Module.UseVelDistribution=false
    Override Seed=false
    Velocity Mode=::NewEnumerator2
    Velocity Speed=&MakeFloatFromVector(
     Vector=&VC_Mask_Vectors(Coro Value:Vector3f=<-400 0 0>, FrontEnd Value:Vector3f=<-1200 0 0>, FN Value:Vector3f=<-1200 0 0>)
     Channel=::NewEnumerator0
    )
   )
   &ShapeLocation(
    Box Size=&MergeHelper_VectorSwitch(0=<1 8 1>, 1=<1 10 1>, 2=<1 6 1>, 3=?, Index=&SpawnGroup(), Num=::NewEnumerator1)
    MaskSpawnGroup=false
    Override Seed=false
    Shape Primitive=::NewEnumerator2
    Surface Only=false
    UseOwnerScale=true
   )
   &CameraOffset(Camera Offset Amount=&MergeHelper_FloatSwitch(0=55, 1=0, 2=100, 3=?, Index=&SpawnGroup(), Num=::NewEnumerator1))
   /SetVariables(Particles.VisibilityTag=&SelectIntFromArray(Direct Array Index=&SpawnGroup(), Int Selection Array=DataInterface() {propertyValues {IntData [0,1,2]}})) {isSetParameters=true, module=SetVariables_#5}
   &DLR_ProductPreviews(GameModeID:NiagaraInt32=$System.GameModeID, bPreviewVehicleSpeed:NiagaraBool=$System.bPreviewVehicleSpeed)
  ]
  particleUpdateScript [
   &ParticleState()
   &ScaleSpriteSize(
    Scale Sprite Size Mode=::NewEnumerator0
    Uniform Scale Factor=&MakeCustomFloatFromBool(
     Bool=$User.bIsFrontendPreview
     True Float=&MergeHelper_FloatSwitch(
      0=&FloatFromCurve(FloatCurve=DataInterface() {propertyValues {Curve {
          defaultValue=1
          keys [{arriveTangent=0.644, interpMode=RCIM_Cubic, leaveTangent=0.644, tangentMode=RCTM_User, time=0.4, value=0.6}, <at 1 2 RCIM_Cubic>]
         }}})
      1=0
      2=&FloatFromCurve(FloatCurve=DataInterface() {propertyValues {Curve {
          defaultValue=1
          keys [{arriveTangent=0.644, interpMode=RCIM_Cubic, leaveTangent=0.644, tangentMode=RCTM_User, time=0.5, value=1}, <at 1 3 RCIM_Cubic>]
         }}})
      3=?
      Index=&SpawnGroup()
      Num=::NewEnumerator1
     )
     False Float=&FloatFromCurve(FloatCurve=DataInterface() {propertyValues {Curve {
         defaultValue=1
         keys [
          {interpMode=RCIM_Cubic, leaveTangentWeight=0.769, tangentMode=RCTM_User, time=0.15, value=0.6}
          {arriveTangent=1.0126, interpMode=RCIM_Cubic, leaveTangent=1.0126, tangentMode=RCTM_User, time=1, value=2.5}
         ]
        }}})
    )
   )
   &SpriteRotationRate(Rotation Rate=&RandomRangeFloat(Minimum=125, Maximum=-125))
   &DynamicMaterialParameters(
    Index 0 Param 1=1
    Index 0 Param 2=&MergeHelper_FloatSwitch(
     0=&FloatFromCurve(FloatCurve=DataInterface() {propertyValues {Curve {defaultValue=0, keys [{tangentMode=RCTM_User, time=0, value=0}, {tangentMode=RCTM_User, time=1, value=0}]}}})
     1=&FloatFromCurve(FloatCurve=DataInterface() {propertyValues {Curve {defaultValue=1, keys [{tangentMode=RCTM_User, time=0, value=0}, {tangentMode=RCTM_User, time=1, value=0.5}]}}})
     2=?
     3=?
     Index=&SpawnGroup()
     Num=::NewEnumerator0
    )
    Index 0 Param 3=&MergeHelper_FloatSwitch(0=1, 1=0, 2=1, 3=?, Index=&SpawnGroup(), Num=::NewEnumerator1)
    Index 0 Param 4=&MergeHelper_FloatSwitch(
     0=&FloatFromCurve(
      FloatCurve=DataInterface() {propertyValues {Curve {defaultValue=1, keys [{tangentMode=RCTM_User, time=0, value=0}, {tangentMode=RCTM_User, time=1, value=1}]}}}
      Scale Curve=0.1
     )
     1=0
     2=&FloatFromCurve(
      FloatCurve=DataInterface() {propertyValues {Curve {defaultValue=1, keys [{tangentMode=RCTM_User, time=0, value=0}, {tangentMode=RCTM_User, time=1, value=1}]}}}
      Scale Curve=0.1
     )
     3=?
     Index=&SpawnGroup()
     Num=::NewEnumerator1
    )
    Index 2 Param 1=0.5
    Index 2 Param 2=0.5
    Index 2 Param 3=&FloatFromCurve(FloatCurve=DataInterface() {propertyValues {Curve {keys [<at 0 0.5>,<at 1 1>]}}}, Scale Curve=0.5)
    Index 2 Param 4=&FloatFromCurve(FloatCurve=DataInterface() {propertyValues {Curve {keys [<at 0 0.5>,<at 1 1>]}}}, Scale Curve=0.5)
    Module.Index0Float0Write=true
    Module.Index0Float1Write=false
    Module.Index0Float2Write=true
    Module.Index0Float3Write=true
    Module.Index2Float0Write=true
    Module.Index2Float1Write=true
    Module.Index2Float2Write=true
    Module.Index2Float3Write=true
    Write Parameter Index 2=::NewEnumerator1
   )
   &InheritSourceMovement(
    Applied Position Scale=&VectorFromFloat(Value=&FloatFromCurve(FloatCurve=DataInterface() {propertyValues {Curve {
         defaultValue=1
         keys [
          {arriveTangent=-1.7885, interpMode=RCIM_Cubic, leaveTangent=-1.7885, tangentMode=RCTM_User, time=0.05, value=1}
          {interpMode=RCIM_Cubic, tangentMode=RCTM_User, time=1, value=0}
         ]
        }}}))
    ApplyForce=false
    ApplyPosition=true
    ApplyPositionFalloff=false
   )
   &GravityForce(Gravity=<0 0 -20>)
   &SolveForcesAndVelocity(Clamp Velocity=false, Limit Acceleration=false)
   &ScaleColor(
    Color Scale:LinearColor=&MergeHelper_ColorSwitch(
     0=&MakeCustomLinearColorFromBool(
      Bool=$User.bIsFrontendPreview
      True Color=&ColorFromCurve(ColorCurve=DataInterface() {propertyValues {
         AlphaCurve {keys [<at 0.0311 0 RCIM_Cubic>, {arriveTangent=0.9091, interpMode=RCIM_Cubic, leaveTangent=0.9091, tangentMode=RCTM_User, time=1, value=2}]}
         BlueCurve {keys [<at 0.0081 0>,<at 0.4206 0.1>,<at 0.8608 0>]}
         GreenCurve {keys [<at 0.0081 0.3317>,<at 0.4206 0.2583>,<at 0.8608 0.0649>]}
         RedCurve {keys [
           {arriveTangent=-1.2682, interpMode=RCIM_Cubic, leaveTangent=-1.2682, tangentMode=RCTM_User, time=0, value=2}
           <at 0.0081 1.9902>
           <at 0.4206 2>
           <at 0.8608 0.5>
          ]}
        }})
      False Color=&ColorFromCurve(ColorCurve=DataInterface() {propertyValues {
         AlphaCurve {keys [
           <at 0 0 RCIM_Cubic>
           {arriveTangent=0.9091, interpMode=RCIM_Cubic, leaveTangent=0.9091, tangentMode=RCTM_User, time=0.1141, value=2}
           <at 0.813 1>
          ]}
         BlueCurve {keys [<at 0 0.2>,<at 0.2097 0.1>,<at 0.7255 0.0154>]}
         GreenCurve {keys [<at 0 1.85>,<at 0.2097 0.205>,<at 0.7255 0.0389>]}
         RedCurve {keys [<at 0 20>,<at 0.2097 1>,<at 0.7255 0.2157>]}
        }})
     )
     1=&ColorFromCurve(ColorCurve=DataInterface() {propertyValues {
        AlphaCurve {keys [<at 0.2723 0 RCIM_Cubic>,<at 1 2 RCIM_Cubic>]}
        BlueCurve {keys [<at 0.7289 0 RCIM_Cubic>,<at 0.994 0>]}
        GreenCurve {keys [<at 0.7289 0.0015 RCIM_Cubic>,<at 0.994 0.0006>]}
        RedCurve {keys [<at 0.7289 0.0625 RCIM_Cubic>,<at 0.994 0.026>]}
       }})
     2=&MakeCustomLinearColorFromBool(
      Bool=$User.bIsFrontendPreview
      True Color=&ColorFromCurve(ColorCurve=DataInterface() {propertyValues {
         AlphaCurve {keys [
           <at 0.1302 0 RCIM_Cubic>
           <at 0.2501 1>
           {arriveTangent=0.9091, interpMode=RCIM_Cubic, leaveTangent=0.9091, tangentMode=RCTM_User, time=0.4621, value=0.5}
           <at 0.8608 0>
          ]}
         BlueCurve {keys [<at 0 0.125>,<at 0.6741 0.05>,<at 1 0.03>]}
         GreenCurve {keys [
           <at 0 0.5208>
           <at 0.6741 0.1174>
           {arriveTangent=-1.4613, interpMode=RCIM_Cubic, leaveTangent=-1.4613, tangentMode=RCTM_User, time=1, value=0.0573}
          ]}
         RedCurve {keys [
           <at 0 2.5>
           <at 0.6741 0.5>
           {arriveTangent=-12.0185, interpMode=RCIM_Cubic, leaveTangent=-12.0185, tangentMode=RCTM_User, time=1, value=0.3}
          ]}
        }})
      False Color=&MakeLinearColorFromVectorAndFloat(
       "Vector (RGB)"=&VectorFromCurve(VectorCurve=DataInterface() {propertyValues {
          XCurve {keys [<at 0 1.1>,<at 0.3888 1>,<at 0.9 0.4>]}
          YCurve {keys [
            {arriveTangent=-0.3418, interpMode=RCIM_Cubic, leaveTangent=-0.3418, tangentMode=RCTM_User, time=0, value=1.1}
            {arriveTangent=-0.7778, interpMode=RCIM_Cubic, leaveTangent=-0.7778, time=0.5149, value=0.6753}
            {arriveTangent=-1.5559, interpMode=RCIM_Cubic, leaveTangent=-1.5559, tangentMode=RCTM_User, time=0.9, value=0.4}
           ]}
          ZCurve {keys [<at 0 1.1>,<at 0.5149 0.7349>,<at 0.9 0.4>]}
         }})
       "Float (Alpha)"=&FloatFromCurve(FloatCurve=DataInterface() {propertyValues {Curve {keys [<at 0.1 0>,<at 0.15 1>,<at 1 1>]}}})
      )
     )
     3=?
     Index=&SpawnGroup()
     Num=::NewEnumerator1
    )
    Scale Mode=3
   )
   &ScaleColor(
    Scale RGB=&MergeHelper_VectorSwitch(
     0=&MakeCustomVectorFromBool(
      Bool=$User.bIsFrontendPreview
      True Vector=&VectorFromCurve(VectorCurve=DataInterface() {propertyValues {
         XCurve {keys [<at 0.1 3 RCIM_Cubic>,<at 1 0.5 RCIM_Cubic>]}
         YCurve {keys [<at 0.1 3 RCIM_Cubic>,<at 1 0.5 RCIM_Cubic>]}
         ZCurve {keys [<at 0.1 3 RCIM_Cubic>,<at 1 0.5 RCIM_Cubic>]}
        }})
      False Vector=&VectorFromCurve(VectorCurve=DataInterface() {propertyValues {
         XCurve {keys [<at 0 1>,<at 0.2 3.5>,<at 0.45 0.8>,<at 1 0.5>]}
         YCurve {keys [<at 0 1>,<at 0.2 1.5>,<at 0.3 1.5>,<at 0.55 1>,<at 1 0.5>]}
         ZCurve {keys [<at 0 1>,<at 0.2 1.5>,<at 0.3 1.5>,<at 0.55 1>,<at 1 0.5>]}
        }})
     )
     1=&VectorFromCurve(VectorCurve=DataInterface() {propertyValues {XCurve {keys [<at 0 0.3>,<at 1 0.1>]}, YCurve {keys [<at 0 0.3>,<at 1 0.1>]}, ZCurve {keys [<at 0 0.3>,<at 1 0.1>]}}})
     2=?
     3=?
     Index=&SpawnGroup()
     Num=::NewEnumerator0
    )
    ScaleA=false
    ScaleRGB=true
   ) {enabled=false, module=ScaleColor001}
  ]
  rendererClasses [&NiagaraSpriteRendererProperties]
  renderers [
   {rendererClass=&NiagaraSpriteRendererProperties}
   {rendererClass=&NiagaraSpriteRendererProperties, rendererIndex=1}
   {rendererClass=&NiagaraSpriteRendererProperties, rendererIndex=2}
  ]
 }
}
summary {
 systemName=NS_Propulsion
 emitters [
  {bEnabled=false, emitterName=NE_VC_Boost_Smoke, rendererClasses [&NiagaraSpriteRendererProperties]}
  {bEnabled=false, emitterName=NE_VC_Boost_Smoke001, rendererClasses [&NiagaraSpriteRendererProperties]}
  {bEnabled=false, emitterName=NE_VC_Boost_Smoke002, rendererClasses [&NiagaraSpriteRendererProperties]}
  {bEnabled=false, emitterName=NE_VC_Boost_Smoke003, rendererClasses [&NiagaraSpriteRendererProperties]}
  {bEnabled=false, emitterName=NE_VC_Boost_Smoke004, rendererClasses [&NiagaraSpriteRendererProperties]}
  {emitterName=NE_VC_Boost_LeftFire, rendererClasses [&NiagaraSpriteRendererProperties]}
  {emitterName=NE_VC_Boost_RightFire, rendererClasses [&NiagaraSpriteRendererProperties]}
 ]
 userVariables [
  User.bIsFrontendPreview {type=NiagaraBool}
  User.bIsPainted {type=NiagaraBool}
  User.blsIsCoroColor {type=NiagaraBool}
  User.bPreviewVehicleSpeed {type=NiagaraBool}
  User.ColorVariantID {type=NiagaraInt32}
  User.GameModeID {type=NiagaraInt32}
  User.PaintedColor {type=LinearColor, description="", defaultValue {struct=LinearColor, value=<rgba 1 1 1 1>}}
  User.PaintedScalar {type=NiagaraFloat, defaultValue=1, description=""}
  User.PortCount {type=NiagaraInt32, defaultValue=2}
 ]
}
